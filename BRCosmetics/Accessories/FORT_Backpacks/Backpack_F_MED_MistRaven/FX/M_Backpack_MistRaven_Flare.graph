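# Unlit masked Niagara material for sprite and mesh flare particles on the MistRaven backpack.
# Polar-coordinate UVs pan two noise textures that vary a spherical center mask and optionally distort the flare texture.
# Color comes from dark/light colors or a curve atlas, with optional particle tint and time-of-day vertical gradient.
# A muzzle-flash blend swaps in a cropped glider flare texture, alongside depth and view-angle fades.
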Material /BRCosmetics/Accessories/FORT_Backpacks/Backpack_F_MED_MistRaven/FX/M_Backpack_MistRaven_Flare
graph {
 properties {
  bAutomaticallySetUsageInEditor=false
  bUsedWithNiagaraMeshParticles=true
  bUsedWithNiagaraSprites=true
  blendMode=BLEND_Masked
  shadingModel=MSM_Unlit
  usedShadingModels=Unlit
 }
 nodes [
  n1 = ParticleColor()
  n2 = ScalarParameter() {defaultValue=0.5, parameterName=CenterNoise_Max}
  n3 = Subtract(A=TextureCoordinate()) {constB=0.5}
  n4 = SquareRoot(None=MaterialFunctionCall(Vector 2=@n3) {materialFunction=/Engine/Functions/Engine_MaterialFunctions02/Utility/VectorLength}.V2 Length)
  n5 = ScalarParameter() {defaultValue=1, parameterName=RadialMult, sliderMax=1}
  n6 = MaterialFunctionCall(Float2=@n3) {materialFunction=&BreakOutFloat2Components}
  n7 = Arctangent2Fast(Y=@n6.G, X=@n6.G)
  n8 = MaterialFunctionCall() {materialFunction=&Pi}
  n9 = Constant() {r=2}
  n10 = AppendVector(
   A=Power(Base=Multiply(A=@n4, B=@n5), Exp=ScalarParameter() {defaultValue=1, parameterName=RadialPower, sliderMax=5})
   B=Divide(
    A=Add(A=@n7, B=@n8.Result)
    B=MaterialFunctionCall(Multiplier=@n9) {materialFunction=&Pi}.Result
   )
  )
  n11 = TextureSample(UVs=Panner(Coordinate=@n10) {speedX=-0.25}) {texture=&AtmosphericCloudNoise01}
  n12 = TextureSample(UVs=Panner(Coordinate=@n10) {speedX=-0.15}) {texture=/BRCosmetics/Weapons/FORT_Melee/Pickaxe_MadameMoth/FX/T_Fibre_Noise}
  n13 = Multiply(A=Saturate(None=Multiply(A=@n11.R) {constB=2}), B=LinearInterpolate(Alpha=@n12.R) {constA=0.25})
  n14 = TextureCoordinate()
  n15 = Multiply(A=LinearInterpolate(Alpha=@n11.R) {constA=-1}, B=ScalarParameter() {defaultValue=1, parameterName=DistortionMult})
  n16 = TextureSampleParameter2D(UVs=LinearInterpolate(A=@n14, B=Add(A=@n14, B=@n15), Alpha=ScalarParameter() {parameterName=UseDistortion})) {parameterName=FlareTexture, texture=&/T_FX_Backpack_MistRaven_RoundedFlare}
  n17 = Multiply(
   A=SphereMask(A=TextureCoordinate(), B=Constant() {r=0.5}, Radius=LinearInterpolate(B=@n2, Alpha=@n13) {constB=0.5}) {attenuationRadius=0.5, hardnessPercent=0}
   B=@n16.G
  )
  n18 = Saturate(None=Add(A=Multiply(A=@n17) {constB=5}, B=ChannelMaskParameter(None=@n16.RGB) {parameterName=FlareChannel}.R))
  n19 = Multiply(A=@n1.A, B=@n18)
  FeatureLevelSwitch(Default=DepthFade(Opacity=@n19, FadeDistance=ScalarParameter() {parameterName=DepthFade}), ES3_1=@n19)
  n20 = VectorParameter() {parameterName=DarkColor, defaultValue=<rgba 0 0 0 1>}
  n21 = VectorParameter() {parameterName=LightColor, defaultValue=<rgba 1 1 1 1>}
  n22 = SphereMask(A=TextureCoordinate(), B=Constant() {r=0.5}, Radius=LinearInterpolate(Alpha=@n12.R)) {attenuationRadius=0.5, hardnessPercent=0}
  n23 = CurveAtlasRowParameter(CurveTime=OneMinus(None=@n22)) {
   atlas=/Game/Athena/LinearColorCurves/CAtlas_Contrail_MadameMoth_Color
   curve=/Game/Athena/LinearColorCurves/CV_Contrail_MadameMoth_Color1
   parameterName=Color
  }
  n24 = VectorParameter() {parameterName=Crop, defaultValue=<rgba 0 0.75 1 1>}
  n25 = TextureSample(UVs=MaterialFunctionCall!(
    UVs=TextureCoordinate()
    TextureIn
    UpperLeftCorner=AppendVector(A=@n24.R, B=@n24.R)
    LowerRightCorner=AppendVector(A=@n24.B, B=@n24.B)
    ExportFloat4
   ) {materialFunction=/Engine/Functions/Engine_MaterialFunctions01/Texturing/TextureCropping}.Crop UVs) {texture=/BRCosmetics/Gadgets/Assets/VinderTech_GliderChute/Glider_MistRaven/FX/T_FX_Glider_MistRaven_DirectionalFlare}
  n26 = LinearInterpolate(
   A=LinearInterpolate(A=LinearInterpolate(A=@n20.RGB, B=@n21.RGB, Alpha=@n22), B=@n23, Alpha=ScalarParameter() {defaultValue=1, parameterName=UseColorCurve})
   B=LinearInterpolate(A=LinearInterpolate(B=@n20.RGB, Alpha=@n25.R), B=@n21.RGB, Alpha=@n25.B)
   Alpha=ScalarParameter() {parameterName=UseMuzzleFlash}
  )
  n27 = LinearInterpolate(A=@n26, B=Multiply(A=@n26, B=@n1.RGB), Alpha=ScalarParameter() {parameterName=MultiplyByParticleColor})
  n28 = VectorParameter() {parameterName=TopColor, defaultValue=<rgba 0 0 0 1>}
  n29 = MaterialFunctionCall(
   Color_Morning=@n28.RGB
   Color_Day=@n28.RGB
   Color_Evening=@n28.RGB
   Color_Night=VectorParameter() {parameterName=TopColorNight, defaultValue=<rgba 0 0 0 1>}.RGB
  ) {materialFunction=&Color_Time_of_Day}
  n30 = VectorParameter() {parameterName=BottomColor, defaultValue=<rgba 1 1 1 1>}
  n31 = Multiply(
   A=Power(
    Base=MaterialFunctionCall() {materialFunction=&LinearGradient}.VGradient
    Exp=ScalarParameter() {defaultValue=1, parameterName=GradientPower}
   )
   B=ScalarParameter() {defaultValue=1, parameterName=GradientMult}
  )
  LinearInterpolate(
   A=@n27
   B=Multiply(A=@n27, B=LinearInterpolate(A=@n29.Result, B=@n30.RGB, Alpha=@n31))
   Alpha=ScalarParameter() {parameterName=MultiplyByGradient}
  )
  Multiply(A=@n22, B=@n16.G)
  n32 = VectorParameter() {parameterName=Crop, defaultValue=<rgba 0 0.75 1 1>}
  n33 = MaterialFunctionCall(UVs=TextureCoordinate(), UpperLeftCorner=AppendVector(A=@n32.R, B=@n32.R), LowerRightCorner=AppendVector(A=@n32.B, B=@n32.B)) {materialFunction=&TextureCropping}
  Multiply(
   A=LinearInterpolate(
    A=@n18
    B=TextureSample(UVs=@n33.Crop UVs) {texture=&T_FX_Glider_MistRaven_DirectionalFlare}.R
    Alpha=ScalarParameter() {parameterName=UseMuzzleFlash}
   )
   B=MaterialFunctionCall("FadeSensitivity (optional)"=ScalarParameter() {parameterName=FadeSensitivity}) {materialFunction=&FadeBasedOnViewAngle}.Result
  )
 ]
}
