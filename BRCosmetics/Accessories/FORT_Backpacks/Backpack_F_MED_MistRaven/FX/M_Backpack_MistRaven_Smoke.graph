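# Smoke material for the MistRaven backpack cosmetic, built for Niagara particles with spline-thickened geometry.
# Panning UV layers feed textures, with color from a curve atlas row, tinted by particle color and time of day.
# Opacity combines particle alpha with masks, view-angle fade and depth fade, behind a shading path switch.
# World position offset layers spline thickening, noise distortion and directional bending, scaled by UV gradients and Niagara dynamic parameters.
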Material /BRCosmetics/Accessories/FORT_Backpacks/Backpack_F_MED_MistRaven/FX/M_Backpack_MistRaven_Smoke
graph {
 expressions {
  MaterialExpressionAdd_0 {
   class=MaterialExpressionAdd
   inputs [
    A {from=MaterialExpressionAdd_23}
    B {from=MaterialExpressionNamedRerouteUsage_21}
   ]
  }
  MaterialExpressionAdd_14 {
   class=MaterialExpressionAdd
   inputs [
    A {from=MaterialExpressionMultiply_25}
    B {from=MaterialExpressionAppendVector_2}
   ]
  }
  MaterialExpressionAdd_2 {
   class=MaterialExpressionAdd
   inputs [
    A {from=MaterialExpressionMultiply_24}
    B {from=MaterialExpressionAppendVector_4}
   ]
  }
  MaterialExpressionAdd_20 {
   class=MaterialExpressionAdd
   inputs [
    A {from=MaterialExpressionNamedRerouteUsage_14}
    B {from=MaterialExpressionNamedRerouteUsage_16}
   ]
   properties {constB=0}
  }
  MaterialExpressionAdd_21 {
   class=MaterialExpressionAdd
   inputs [
    A {from=MaterialExpressionAppendVector_35}
    B {from=MaterialExpressionPanner_9}
   ]
  }
  MaterialExpressionAdd_22 {
   class=MaterialExpressionAdd
   inputs [
    A {from=MaterialExpressionMultiply_77}
    B {from=MaterialExpressionMultiply_61}
   ]
   properties {constB=0}
  }
  MaterialExpressionAdd_23 {
   class=MaterialExpressionAdd
   inputs [
    A {from=MaterialExpressionAdd_20}
    B {from=MaterialExpressionNamedRerouteUsage_18}
   ]
   properties {constB=0}
  }
  MaterialExpressionAdd_24 {
   class=MaterialExpressionAdd
   inputs [
    A {from=MaterialExpressionPanner_10}
    B {from=MaterialExpressionAppendVector_44}
   ]
  }
  MaterialExpressionAdd_25 {
   class=MaterialExpressionAdd
   inputs [
    A {from=MaterialExpressionPanner_11}
    B {from=MaterialExpressionAppendVector_44}
   ]
  }
  MaterialExpressionAdd_31 {
   class=MaterialExpressionAdd
   inputs [
    A {from=MaterialExpressionAdd_21}
    B {from=""}
   ]
   properties {constB=0}
  }
  MaterialExpressionAdd_32 {
   class=MaterialExpressionAdd
   inputs [
    A {from=MaterialExpressionAdd_24}
    B {from=MaterialExpressionNamedRerouteUsage_24}
   ]
  }
  MaterialExpressionAdd_33 {
   class=MaterialExpressionAdd
   inputs [
    A {from=MaterialExpressionAdd_25}
    B {from=MaterialExpressionNamedRerouteUsage_25}
   ]
  }
  MaterialExpressionAdd_5 {
   class=MaterialExpressionAdd
   inputs [
    A {from=MaterialExpressionAppendVector_0}
    B {from=MaterialExpressionAppendVector_5}
   ]
  }
  MaterialExpressionAdd_6 {
   class=MaterialExpressionAdd
   inputs [
    A {from=MaterialExpressionNamedRerouteUsage_5}
    B {from=MaterialExpressionAppendVector_6}
   ]
  }
  MaterialExpressionAdd_9 {
   class=MaterialExpressionAdd
   inputs [
    A {from=MaterialExpressionAppendVector_3}
    B {from=MaterialExpressionAppendVector_7}
   ]
  }
  MaterialExpressionAppendVector_0 {
   class=MaterialExpressionAppendVector
   inputs [
    A {from=MaterialExpressionPower_1, output=""}
    B {from=MaterialExpressionComponentMask_2, output=""}
   ]
  }
  MaterialExpressionAppendVector_2 {
   class=MaterialExpressionAppendVector
   inputs [
    A {from=MaterialExpressionConstant_8, output=""}
    B {from=MaterialExpressionConstant_1, output=""}
   ]
  }
  MaterialExpressionAppendVector_3 {
   class=MaterialExpressionAppendVector
   inputs [
    A {from=MaterialExpressionPower_0, output=""}
    B {from=MaterialExpressionComponentMask_3, output=""}
   ]
  }
  MaterialExpressionAppendVector_34 {
   class=MaterialExpressionAppendVector
   inputs [
    A {from=MaterialExpressionVectorParameter_13, output=WPO Depth U Pan}
    B {from=MaterialExpressionVectorParameter_13, output=WPO Depth U Pan}
   ]
  }
  MaterialExpressionAppendVector_35 {
   class=MaterialExpressionAppendVector
   inputs [
    A {from=MaterialExpressionReroute_11, output=""}
    B {from=MaterialExpressionReroute_11, output=""}
   ]
  }
  MaterialExpressionAppendVector_36 {
   class=MaterialExpressionAppendVector
   inputs [
    A {from=MaterialExpressionVectorParameter_13, output=WPO Depth U}
    B {from=MaterialExpressionVectorParameter_13, output=WPO Depth U}
   ]
  }
  MaterialExpressionAppendVector_37 {
   class=MaterialExpressionAppendVector
   inputs [
    A {from=MaterialExpressionVectorParameter_14, output=B}
    B {from=MaterialExpressionVectorParameter_14, output=B}
   ]
  }
  MaterialExpressionAppendVector_38 {
   class=MaterialExpressionAppendVector
   inputs [
    A {from=MaterialExpressionVectorParameter_14, output=R}
    B {from=MaterialExpressionVectorParameter_14, output=R}
   ]
  }
  MaterialExpressionAppendVector_39 {
   class=MaterialExpressionAppendVector
   inputs [
    A {from=MaterialExpressionComponentMask_12, output=""}
    B {from=MaterialExpressionPower_16, output=""}
   ]
  }
  MaterialExpressionAppendVector_4 {
   class=MaterialExpressionAppendVector
   inputs [
    A {from=MaterialExpressionConstant_5, output=""}
    B {from=MaterialExpressionConstant_0, output=""}
   ]
  }
  MaterialExpressionAppendVector_40 {
   class=MaterialExpressionAppendVector
   inputs [
    A {from=MaterialExpressionMaterialFunctionCall_17, output=R}
    B {from=MaterialExpressionMaterialFunctionCall_17, output=R}
   ]
  }
  MaterialExpressionAppendVector_41 {
   class=MaterialExpressionAppendVector
   inputs [
    A {from=MaterialExpressionClamp_2, output=""}
    B {from=MaterialExpressionClamp_3, output=""}
   ]
  }
  MaterialExpressionAppendVector_42 {
   class=MaterialExpressionAppendVector
   inputs [
    A {from=MaterialExpressionVectorParameter_15, output=B}
    B {from=MaterialExpressionVectorParameter_15, output=B}
   ]
  }
  MaterialExpressionAppendVector_43 {
   class=MaterialExpressionAppendVector
   inputs [
    A {from=MaterialExpressionVectorParameter_15, output=R}
    B {from=MaterialExpressionVectorParameter_15, output=R}
   ]
  }
  MaterialExpressionAppendVector_44 {
   class=MaterialExpressionAppendVector
   inputs [
    A {from=MaterialExpressionScalarParameter_58, output=""}
    B {from=MaterialExpressionScalarParameter_58, output=""}
   ]
  }
  MaterialExpressionAppendVector_5 {
   class=MaterialExpressionAppendVector
   inputs [
    A {from=MaterialExpressionConstant_12, output=""}
    B {from=MaterialExpressionConstant_9, output=""}
   ]
  }
  MaterialExpressionAppendVector_6 {
   class=MaterialExpressionAppendVector
   inputs [
    A {from=MaterialExpressionConstant_13, output=""}
    B {from=MaterialExpressionConstant_10, output=""}
   ]
  }
  MaterialExpressionAppendVector_7 {
   class=MaterialExpressionAppendVector
   inputs [
    A {from=MaterialExpressionConstant_14, output=""}
    B {from=MaterialExpressionConstant_11, output=""}
   ]
  }
  MaterialExpressionChannelMaskParameter_0 {
   class=MaterialExpressionChannelMaskParameter
   inputs [None {from=MaterialExpressionTextureSampleParameter2D_0, output=RGBA}]
   properties {group=Tex, parameterName=Mask}
  }
  MaterialExpressionClamp_0 {
   class=MaterialExpressionClamp
   inputs [
    None {from=MaterialExpressionLinearInterpolate_0, output=""}
    Min {from=MaterialExpressionScalarParameter_14, output=""}
    Max {from="", output=""}
   ]
   properties {minDefault=0.1}
  }
  MaterialExpressionClamp_2 {
   class=MaterialExpressionClamp
   inputs [
    None {from=MaterialExpressionAppendVector_40, output=""}
    Min {from=MaterialExpressionMultiply_68, output=""}
    Max {from=MaterialExpressionScalarParameter_49, output=""}
   ]
   properties {maxDefault=1000, minDefault=-1000}
  }
  MaterialExpressionClamp_3 {
   class=MaterialExpressionClamp
   inputs [
    None {from=MaterialExpressionMaterialFunctionCall_17, output=B}
    Min {from=MaterialExpressionMultiply_69, output=""}
    Max {from=MaterialExpressionScalarParameter_51, output=""}
   ]
   properties {minDefault=-1}
  }
  MaterialExpressionComponentMask_0 {class=MaterialExpressionComponentMask, inputs [None {from=MaterialExpressionNamedRerouteUsage_10}], properties {r=true}}
  MaterialExpressionComponentMask_1 {class=MaterialExpressionComponentMask, inputs [None {from=MaterialExpressionMultiply_4}], properties {r=true}}
  MaterialExpressionComponentMask_10 {class=MaterialExpressionComponentMask, inputs [None {from=MaterialExpressionTextureCoordinate_18}], properties {g=true}}
  MaterialExpressionComponentMask_11 {class=MaterialExpressionComponentMask, inputs [None {from=MaterialExpressionMultiply_21}], properties {r=true}}
  MaterialExpressionComponentMask_12 {class=MaterialExpressionComponentMask, inputs [None {from=MaterialExpressionTextureCoordinate_18}], properties {r=true}}
  MaterialExpressionComponentMask_13 {class=MaterialExpressionComponentMask, inputs [None {from=MaterialExpressionTextureCoordinate_20}], properties {g=true}}
  MaterialExpressionComponentMask_14 {class=MaterialExpressionComponentMask, inputs [None {from=MaterialExpressionVectorParameter_4, output=RGB}], properties {g=true, r=true}}
  MaterialExpressionComponentMask_15 {class=MaterialExpressionComponentMask, inputs [None {from=MaterialExpressionVectorParameter_5, output=RGB}], properties {g=true, r=true}}
  MaterialExpressionComponentMask_16 {class=MaterialExpressionComponentMask, inputs [None {from=MaterialExpressionOneMinus_5}], properties {g=true}}
  MaterialExpressionComponentMask_18 {class=MaterialExpressionComponentMask, inputs [None {from=MaterialExpressionTextureCoordinate_2}], properties {g=true}}
  MaterialExpressionComponentMask_2 {class=MaterialExpressionComponentMask, inputs [None {from=MaterialExpressionMultiply_4}], properties {g=true}}
  MaterialExpressionComponentMask_3 {class=MaterialExpressionComponentMask, inputs [None {from=MaterialExpressionMultiply_21}], properties {g=true}}
  MaterialExpressionComponentMask_4 {class=MaterialExpressionComponentMask, inputs [None {from=MaterialExpressionNamedRerouteUsage_9}], properties {r=true}}
  MaterialExpressionComponentMask_5 {class=MaterialExpressionComponentMask, inputs [None {from=MaterialExpressionNamedRerouteUsage_12}], properties {r=true}}
  MaterialExpressionComponentMask_6 {class=MaterialExpressionComponentMask, inputs [None {from=MaterialExpressionVectorParameter_1, output=RGB}], properties {g=true, r=true}}
  MaterialExpressionComponentMask_7 {class=MaterialExpressionComponentMask, inputs [None {from=MaterialExpressionNamedRerouteUsage_7}], properties {r=true}}
  MaterialExpressionComponentMask_8 {class=MaterialExpressionComponentMask, inputs [None {from=MaterialExpressionVectorParameter_3, output=RGB}], properties {g=true, r=true}}
  MaterialExpressionComponentMask_9 {class=MaterialExpressionComponentMask, inputs [None {from=MaterialExpressionTextureCoordinate_22}], properties {g=true}}
  MaterialExpressionConstant2Vector_1 {class=MaterialExpressionConstant2Vector, properties {g=0.2, r=1}}
  MaterialExpressionConstant_0 {class=MaterialExpressionConstant}
  MaterialExpressionConstant_1 {class=MaterialExpressionConstant}
  MaterialExpressionConstant_10 {class=MaterialExpressionConstant}
  MaterialExpressionConstant_11 {class=MaterialExpressionConstant}
  MaterialExpressionConstant_12 {class=MaterialExpressionConstant}
  MaterialExpressionConstant_13 {class=MaterialExpressionConstant}
  MaterialExpressionConstant_14 {class=MaterialExpressionConstant}
  MaterialExpressionConstant_23 {class=MaterialExpressionConstant, properties {r=1.8}}
  MaterialExpressionConstant_24 {class=MaterialExpressionConstant, properties {r=0.9}}
  MaterialExpressionConstant_25 {class=MaterialExpressionConstant}
  MaterialExpressionConstant_26 {class=MaterialExpressionConstant}
  MaterialExpressionConstant_27 {class=MaterialExpressionConstant, properties {r=1}}
  MaterialExpressionConstant_28 {class=MaterialExpressionConstant, properties {r=1}}
  MaterialExpressionConstant_29 {class=MaterialExpressionConstant}
  MaterialExpressionConstant_3 {class=MaterialExpressionConstant}
  MaterialExpressionConstant_33 {class=MaterialExpressionConstant, properties {r=2}}
  MaterialExpressionConstant_40 {class=MaterialExpressionConstant, properties {r=0.25}}
  MaterialExpressionConstant_5 {class=MaterialExpressionConstant}
  MaterialExpressionConstant_6 {class=MaterialExpressionConstant, properties {r=1}}
  MaterialExpressionConstant_7 {class=MaterialExpressionConstant}
  MaterialExpressionConstant_8 {class=MaterialExpressionConstant}
  MaterialExpressionConstant_9 {class=MaterialExpressionConstant}
  MaterialExpressionConvert_0 {
   class=MaterialExpressionConvert
   inputs [
    None {from=MaterialExpressionOneMinus_4}
    None {from=MaterialExpressionMaterialFunctionCall_5, output=G}
   ]
   properties {
    nodeName=Make Float2
    convertInputs [
     {
      type=Scalar
      expressionInput {
       expression=&self:MaterialExpressionOneMinus_4
       inputName=None
       mask=0
       maskA=0
       maskB=0
       maskG=0
       maskR=0
       outputIndex=0
      }
     }
     {
      type=Scalar
      expressionInput {
       expression=&self:MaterialExpressionMaterialFunctionCall_5
       inputName=None
       mask=0
       maskA=0
       maskB=0
       maskG=0
       maskR=0
       outputIndex=1
      }
     }
    ]
    convertOutputs [{type=Vector2}]
   }
  }
  MaterialExpressionCurveAtlasRowParameter_0 {
   class=MaterialExpressionCurveAtlasRowParameter
   inputs [CurveTime {from=MaterialExpressionLinearInterpolate_1, output=""}]
   properties {
    atlas=/Game/Athena/LinearColorCurves/CAtlas_Contrail_MadameMoth_Color
    curve=/Game/Athena/LinearColorCurves/CV_Contrail_MadameMoth_Color1
    parameterName=Color
   }
  }
  MaterialExpressionDepthFade_0 {
   class=MaterialExpressionDepthFade
   inputs [
    Opacity {from=MaterialExpressionMultiply_6, output=""}
    FadeDistance {from=MaterialExpressionScalarParameter_13, output=""}
   ]
   properties {}
  }
  MaterialExpressionDynamicParameter_0 {
   class=MaterialExpressionDynamicParameter
   inputs []
   properties {desc=Parameters set by Niagara, parameterIndex=2, paramNames [Add Noise Lerp, AddedWPOMult, "", ""]}
  }
  MaterialExpressionDynamicParameter_13 {class=MaterialExpressionDynamicParameter, inputs [], properties {parameterIndex=2, paramNames [Add Noise Lerp, AddedWPOMult, "", ""]}}
  MaterialExpressionDynamicParameter_2 {
   class=MaterialExpressionDynamicParameter
   inputs []
   properties {defaultValue=<rgba 0 1 1 1>, paramNames [Random Offset, Spline Width Mult, Dissolve, Move UV]}
  }
  MaterialExpressionDynamicParameter_3 {
   class=MaterialExpressionDynamicParameter
   inputs []
   properties {parameterIndex=1, defaultValue=<rgba 0 10 1 1>, paramNames [Random Offset, Noise Intensty, Dissolve, Dissolve Edge Intensity]}
  }
  MaterialExpressionDynamicParameter_4 {
   class=MaterialExpressionDynamicParameter
   inputs []
   properties {desc=Parameters set by Niagara, parameterIndex=3, defaultValue=<rgba 0 1 1 1>, paramNames [X,Y,Z,Intensity]}
  }
  MaterialExpressionLinearInterpolate_0 {
   class=MaterialExpressionLinearInterpolate
   inputs [
    A {from=""}
    B {from=""}
    Alpha {from=MaterialExpressionTextureSample_0, output=R}
   ]
   properties {constA=-0.8}
  }
  MaterialExpressionLinearInterpolate_1 {
   class=MaterialExpressionLinearInterpolate
   inputs [
    A {from=MaterialExpressionComponentMask_5}
    B {from=MaterialExpressionLinearInterpolate_2}
    Alpha {from=MaterialExpressionScalarParameter_21}
   ]
  }
  MaterialExpressionLinearInterpolate_10 {
   class=MaterialExpressionLinearInterpolate
   inputs [
    A {from=""}
    B {from=MaterialExpressionMultiply_19}
    Alpha {from=MaterialExpressionScalarParameter_6}
   ]
  }
  MaterialExpressionLinearInterpolate_11 {
   class=MaterialExpressionLinearInterpolate
   inputs [
    A {from=""}
    B {from=MaterialExpressionDynamicParameter_13, output=Add Noise Lerp}
    Alpha {from=MaterialExpressionScalarParameter_9}
   ]
   properties {constA=1}
  }
  MaterialExpressionLinearInterpolate_12 {
   class=MaterialExpressionLinearInterpolate
   inputs [
    A {from=""}
    B {from=MaterialExpressionOneMinus_3}
    Alpha {from=MaterialExpressionScalarParameter_15}
   ]
   properties {constA=1}
  }
  MaterialExpressionLinearInterpolate_14 {
   class=MaterialExpressionLinearInterpolate
   inputs [
    A {from=MaterialExpressionScalarParameter_57}
    B {from=""}
    Alpha {from=MaterialExpressionTextureSample_2, output=R}
   ]
   properties {constB=-1}
  }
  MaterialExpressionLinearInterpolate_15 {
   class=MaterialExpressionLinearInterpolate
   inputs [
    A {from=""}
    B {from=""}
    Alpha {from=MaterialExpressionTextureSample_5, output=R}
   ]
   properties {constA=-1}
  }
  MaterialExpressionLinearInterpolate_16 {
   class=MaterialExpressionLinearInterpolate
   inputs [
    A {from=""}
    B {from=MaterialExpressionPower_15}
    Alpha {from=MaterialExpressionScalarParameter_45}
   ]
   properties {constA=1}
  }
  MaterialExpressionLinearInterpolate_17 {
   class=MaterialExpressionLinearInterpolate
   inputs [
    A {from=MaterialExpressionScalarParameter_44}
    B {from=MaterialExpressionDynamicParameter_3, output=Noise Intensty}
    Alpha {from=MaterialExpressionScalarParameter_46}
   ]
  }
  MaterialExpressionLinearInterpolate_18 {
   class=MaterialExpressionLinearInterpolate
   inputs [
    A {from=""}
    B {from=MaterialExpressionAppendVector_41}
    Alpha {from=MaterialExpressionMultiply_67}
   ]
  }
  MaterialExpressionLinearInterpolate_19 {
   class=MaterialExpressionLinearInterpolate
   inputs [
    A {from=""}
    B {from=""}
    Alpha {from=MaterialExpressionTextureSample_6, output=R}
   ]
   properties {constA=-1}
  }
  MaterialExpressionLinearInterpolate_2 {
   class=MaterialExpressionLinearInterpolate
   inputs [
    A {from=""}
    B {from=MaterialExpressionClamp_0}
    Alpha {from=MaterialExpressionMultiply_22}
   ]
   properties {constA=1}
  }
  MaterialExpressionLinearInterpolate_28 {
   class=MaterialExpressionLinearInterpolate
   inputs [
    A {from=MaterialExpressionScalarParameter_72}
    B {from=MaterialExpressionMultiply_90}
    Alpha {from=MaterialExpressionScalarParameter_76}
   ]
  }
  MaterialExpressionLinearInterpolate_29 {
   class=MaterialExpressionLinearInterpolate
   inputs [
    A {from=MaterialExpressionScalarParameter_73}
    B {from=MaterialExpressionMultiply_91}
    Alpha {from=MaterialExpressionScalarParameter_76}
   ]
  }
  MaterialExpressionLinearInterpolate_3 {
   class=MaterialExpressionLinearInterpolate
   inputs [
    A {from=""}
    B {from=MaterialExpressionLinearInterpolate_7}
    Alpha {from=MaterialExpressionScalarParameter_1}
   ]
   properties {constA=1}
  }
  MaterialExpressionLinearInterpolate_4 {
   class=MaterialExpressionLinearInterpolate
   inputs [
    A {from=MaterialExpressionScalarParameter_2}
    B {from=MaterialExpressionScalarParameter_3}
    Alpha {from=MaterialExpressionPower_10}
   ]
   properties {constA=0.1, constB=2}
  }
  MaterialExpressionLinearInterpolate_5 {
   class=MaterialExpressionLinearInterpolate
   inputs [
    A {from=MaterialExpressionScalarParameter_4}
    B {from=MaterialExpressionScalarParameter_5}
    Alpha {from=MaterialExpressionPower_9}
   ]
   properties {constA=-2, constB=2}
  }
  MaterialExpressionLinearInterpolate_6 {
   class=MaterialExpressionLinearInterpolate
   inputs [
    A {from=""}
    B {from=""}
    Alpha {from=MaterialExpressionTextureSample_1, output=R}
   ]
   properties {constB=1.5}
  }
  MaterialExpressionLinearInterpolate_7 {
   class=MaterialExpressionLinearInterpolate
   inputs [
    A {from=MaterialExpressionMultiply_8}
    B {from=MaterialExpressionCurveAtlasRowParameter_0}
    Alpha {from=MaterialExpressionScalarParameter_22}
   ]
  }
  MaterialExpressionLinearInterpolate_9 {
   class=MaterialExpressionLinearInterpolate
   inputs [
    A {from=MaterialExpressionConvert_0}
    B {from=MaterialExpressionTextureCoordinate_0}
    Alpha {from=MaterialExpressionScalarParameter_10}
   ]
  }
  MaterialExpressionMaterialFunctionCall_0 {
   class=MaterialExpressionMaterialFunctionCall
   inputs [
    Color_Morning {from=MaterialExpressionConstant_6}
    Color_Day {from=MaterialExpressionConstant_6}
    Color_Evening {from=MaterialExpressionConstant_6}
    Color_Night {from=MaterialExpressionVectorParameter_2, output=RGB}
   ]
   properties {materialFunction=&Color_Time_of_Day}
  }
  MaterialExpressionMaterialFunctionCall_1 {
   class=MaterialExpressionMaterialFunctionCall
   inputs ["FadeSensitivity (optional)" {from=MaterialExpressionScalarParameter_12}]
   properties {materialFunction=&FadeBasedOnViewAngle}
  }
  MaterialExpressionMaterialFunctionCall_12 {
   class=MaterialExpressionMaterialFunctionCall
   inputs [
    R {from=MaterialExpressionConstant_26}
    G {from=MaterialExpressionMultiply_53}
    B {from=MaterialExpressionConstant_26}
    A {from=""}
   ]
   properties {materialFunction=&AppendMany}
  }
  MaterialExpressionMaterialFunctionCall_13 {
   class=MaterialExpressionMaterialFunctionCall
   inputs [
    R {from=MaterialExpressionMultiply_66}
    G {from=MaterialExpressionConstant_25}
    B {from=MaterialExpressionMultiply_63}
    A {from=""}
   ]
   properties {materialFunction=&AppendMany}
  }
  MaterialExpressionMaterialFunctionCall_16 {
   class=MaterialExpressionMaterialFunctionCall
   inputs [
    X {from=MaterialExpressionDynamicParameter_4, output=X}
    Y {from=MaterialExpressionDynamicParameter_4, output=X}
    Z {from=MaterialExpressionDynamicParameter_4, output=X}
   ]
   properties {materialFunction=&MakeFloat3}
  }
  MaterialExpressionMaterialFunctionCall_17 {
   class=MaterialExpressionMaterialFunctionCall
   inputs [Float3 {from=MaterialExpressionSubtract_0}]
   properties {materialFunction=&BreakOutFloat3Components}
  }
  MaterialExpressionMaterialFunctionCall_2 {
   class=MaterialExpressionMaterialFunctionCall
   inputs [Emissive {from=MaterialExpressionMultiply_18}]
   properties {materialFunction=&MF_FX_EmissiveAdjustments}
  }
  MaterialExpressionMaterialFunctionCall_22 {
   class=MaterialExpressionMaterialFunctionCall
   inputs [
    UVs {from=""}
    Offset {from=""}
    RotationCenter {from=""}
    RotationAngle {from=MaterialExpressionConstant_40}
   ]
   properties {materialFunction=/FortUILibrary/MaterialFunctions/Gradients/MF_UI_ReflectedGradient}
  }
  MaterialExpressionMaterialFunctionCall_3 {
   class=MaterialExpressionMaterialFunctionCall
   inputs [
    XYZ {from=""}
    XY {from=MaterialExpressionTextureCoordinate_0}
   ]
   properties {materialFunction=&Swizzle}
  }
  MaterialExpressionMaterialFunctionCall_4 {
   class=MaterialExpressionMaterialFunctionCall
   inputs [
    WidthBase {from=MaterialExpressionLinearInterpolate_28}
    WidthTip {from=MaterialExpressionLinearInterpolate_29}
    WorldPosition {from=""}
    UVs for Projection {from=""}
    UVs for Thickness {from=""}
    UVs for texturing {from=""}
    Expand U or V UV Channel {from=""}
    DeriveNormalZ {from=""}
    AngleCorrectedNormals {from=""}
    AdditionalNormal {from=""}
    FlattenPixelNormal {from=""}
    UVs For NormalShading {from=""}
   ]
   properties {materialFunction=&SplineThicken}
  }
  MaterialExpressionMaterialFunctionCall_5 {
   class=MaterialExpressionMaterialFunctionCall
   inputs [Float2 {from=MaterialExpressionMaterialFunctionCall_3, output=YX}]
   properties {materialFunction=&BreakOutFloat2Components}
  }
  MaterialExpressionMultiply_0 {
   class=MaterialExpressionMultiply
   inputs [
    A {from=MaterialExpressionSubtract_1}
    B {from=MaterialExpressionSaturate_0}
   ]
  }
  MaterialExpressionMultiply_1 {
   class=MaterialExpressionMultiply
   inputs [
    A {from=MaterialExpressionScalarParameter_74}
    B {from=MaterialExpressionMaterialFunctionCall_22, output=Result}
   ]
  }
  MaterialExpressionMultiply_10 {
   class=MaterialExpressionMultiply
   inputs [
    A {from=MaterialExpressionVectorParameter_0, output=RGB}
    B {from=MaterialExpressionSaturate_2}
   ]
  }
  MaterialExpressionMultiply_11 {
   class=MaterialExpressionMultiply
   inputs [
    A {from=MaterialExpressionOneMinus_6}
    B {from=MaterialExpressionScalarParameter_18}
   ]
   properties {constB=5}
  }
  MaterialExpressionMultiply_16 {
   class=MaterialExpressionMultiply
   inputs [
    A {from=MaterialExpressionComponentMask_4}
    B {from=""}
   ]
   properties {constB=10}
  }
  MaterialExpressionMultiply_18 {
   class=MaterialExpressionMultiply
   inputs [
    A {from=MaterialExpressionMaterialFunctionCall_0, output=Result}
    B {from=MaterialExpressionMultiply_7}
   ]
  }
  MaterialExpressionMultiply_19 {
   class=MaterialExpressionMultiply
   inputs [
    A {from=MaterialExpressionReroute_1}
    B {from=MaterialExpressionSaturate_3}
   ]
  }
  MaterialExpressionMultiply_2 {
   class=MaterialExpressionMultiply
   inputs [
    A {from=MaterialExpressionMultiply_0}
    B {from=MaterialExpressionParticleColor_0, output=A}
   ]
  }
  MaterialExpressionMultiply_21 {
   class=MaterialExpressionMultiply
   inputs [
    A {from=MaterialExpressionNamedRerouteUsage_4}
    B {from=MaterialExpressionComponentMask_8}
   ]
  }
  MaterialExpressionMultiply_22 {
   class=MaterialExpressionMultiply
   inputs [
    A {from=MaterialExpressionScalarParameter_0}
    B {from=MaterialExpressionLinearInterpolate_11}
   ]
  }
  MaterialExpressionMultiply_23 {
   class=MaterialExpressionMultiply
   inputs [
    A {from=MaterialExpressionMultiply_9}
    B {from=MaterialExpressionLinearInterpolate_12}
   ]
  }
  MaterialExpressionMultiply_24 {
   class=MaterialExpressionMultiply
   inputs [
    A {from=MaterialExpressionNamedRerouteUsage_11}
    B {from=MaterialExpressionComponentMask_14}
   ]
  }
  MaterialExpressionMultiply_25 {
   class=MaterialExpressionMultiply
   inputs [
    A {from=MaterialExpressionNamedRerouteUsage_6}
    B {from=MaterialExpressionComponentMask_15}
   ]
  }
  MaterialExpressionMultiply_28 {
   class=MaterialExpressionMultiply
   inputs [
    A {from=MaterialExpressionMultiply_23}
    B {from=MaterialExpressionSaturate_4}
   ]
  }
  MaterialExpressionMultiply_3 {
   class=MaterialExpressionMultiply
   inputs [
    A {from=MaterialExpressionPower_3}
    B {from=""}
   ]
   properties {constB=5}
  }
  MaterialExpressionMultiply_31 {
   class=MaterialExpressionMultiply
   inputs [
    A {from=MaterialExpressionMultiply_10}
    B {from=MaterialExpressionMultiply_39}
   ]
  }
  MaterialExpressionMultiply_32 {
   class=MaterialExpressionMultiply
   inputs [
    A {from=MaterialExpressionScalarParameter_28}
    B {from=MaterialExpressionDynamicParameter_2, output=Spline Width Mult}
   ]
  }
  MaterialExpressionMultiply_37 {
   class=MaterialExpressionMultiply
   inputs [
    A {from=MaterialExpressionMaterialFunctionCall_4, output=WorldPositionOffset}
    B {from=MaterialExpressionMultiply_32}
   ]
  }
  MaterialExpressionMultiply_39 {
   class=MaterialExpressionMultiply
   inputs [
    A {from=MaterialExpressionDynamicParameter_0, output=AddedWPOMult}
    B {from=MaterialExpressionScalarParameter_16}
   ]
  }
  MaterialExpressionMultiply_4 {
   class=MaterialExpressionMultiply
   inputs [
    A {from=MaterialExpressionNamedRerouteUsage_3}
    B {from=MaterialExpressionComponentMask_6}
   ]
  }
  MaterialExpressionMultiply_40 {
   class=MaterialExpressionMultiply
   inputs [
    A {from=MaterialExpressionPower_7}
    B {from=MaterialExpressionScalarParameter_17}
   ]
  }
  MaterialExpressionMultiply_5 {
   class=MaterialExpressionMultiply
   inputs [
    A {from=MaterialExpressionTime_0}
    B {from=MaterialExpressionScalarParameter_11}
   ]
  }
  MaterialExpressionMultiply_52 {
   class=MaterialExpressionMultiply
   inputs [
    A {from=MaterialExpressionTextureCoordinate_19}
    B {from=MaterialExpressionAppendVector_36}
   ]
  }
  MaterialExpressionMultiply_53 {
   class=MaterialExpressionMultiply
   inputs [
    A {from=MaterialExpressionLinearInterpolate_14}
    B {from=MaterialExpressionScalarParameter_43}
   ]
   properties {constB=6}
  }
  MaterialExpressionMultiply_54 {
   class=MaterialExpressionMultiply
   inputs [
    A {from=MaterialExpressionMaterialFunctionCall_12, output=RGB}
    B {from=MaterialExpressionSaturate_11}
   ]
  }
  MaterialExpressionMultiply_56 {
   class=MaterialExpressionMultiply
   inputs [
    A {from=MaterialExpressionReroute_12}
    B {from=MaterialExpressionAppendVector_38}
   ]
  }
  MaterialExpressionMultiply_6 {
   class=MaterialExpressionMultiply
   inputs [
    A {from=MaterialExpressionMultiply_28}
    B {from=MaterialExpressionMaterialFunctionCall_1, output=Result}
   ]
  }
  MaterialExpressionMultiply_60 {
   class=MaterialExpressionMultiply
   inputs [
    A {from=MaterialExpressionMaterialFunctionCall_13, output=RGB}
    B {from=MaterialExpressionLinearInterpolate_17}
   ]
  }
  MaterialExpressionMultiply_61 {
   class=MaterialExpressionMultiply
   inputs [
    A {from=MaterialExpressionMultiply_60}
    B {from=MaterialExpressionMultiply_62}
   ]
  }
  MaterialExpressionMultiply_62 {
   class=MaterialExpressionMultiply
   inputs [
    A {from=MaterialExpressionPower_14}
    B {from=MaterialExpressionLinearInterpolate_16}
   ]
  }
  MaterialExpressionMultiply_63 {
   class=MaterialExpressionMultiply
   inputs [
    A {from=MaterialExpressionLinearInterpolate_19}
    B {from=MaterialExpressionScalarParameter_47}
   ]
   properties {constB=0.5}
  }
  MaterialExpressionMultiply_64 {
   class=MaterialExpressionMultiply
   inputs [
    A {from=MaterialExpressionPower_17}
    B {from=MaterialExpressionPower_18}
   ]
  }
  MaterialExpressionMultiply_65 {
   class=MaterialExpressionMultiply
   inputs [
    A {from=MaterialExpressionMultiply_64}
    B {from=MaterialExpressionConstant_33}
   ]
   properties {constB=4}
  }
  MaterialExpressionMultiply_66 {
   class=MaterialExpressionMultiply
   inputs [
    A {from=MaterialExpressionLinearInterpolate_15}
    B {from=MaterialExpressionScalarParameter_48}
   ]
  }
  MaterialExpressionMultiply_67 {
   class=MaterialExpressionMultiply
   inputs [
    A {from=MaterialExpressionDynamicParameter_4, output=Intensity}
    B {from=MaterialExpressionPower_19}
   ]
  }
  MaterialExpressionMultiply_68 {
   class=MaterialExpressionMultiply
   inputs [
    A {from=MaterialExpressionScalarParameter_49}
    B {from=""}
   ]
   properties {constB=-1}
  }
  MaterialExpressionMultiply_69 {
   class=MaterialExpressionMultiply
   inputs [
    A {from=MaterialExpressionScalarParameter_51}
    B {from=""}
   ]
   properties {constB=-1}
  }
  MaterialExpressionMultiply_7 {
   class=MaterialExpressionMultiply
   inputs [
    A {from=MaterialExpressionLinearInterpolate_3}
    B {from=MaterialExpressionParticleColor_0, output=RGB}
   ]
  }
  MaterialExpressionMultiply_74 {
   class=MaterialExpressionMultiply
   inputs [
    A {from=MaterialExpressionReroute_12}
    B {from=MaterialExpressionAppendVector_43}
   ]
  }
  MaterialExpressionMultiply_77 {
   class=MaterialExpressionMultiply
   inputs [
    A {from=MaterialExpressionMultiply_54}
    B {from=MaterialExpressionLinearInterpolate_17}
   ]
  }
  MaterialExpressionMultiply_8 {
   class=MaterialExpressionMultiply
   inputs [
    A {from=MaterialExpressionLinearInterpolate_2}
    B {from=MaterialExpressionCurveAtlasRowParameter_0}
   ]
  }
  MaterialExpressionMultiply_9 {
   class=MaterialExpressionMultiply
   inputs [
    A {from=MaterialExpressionMultiply_2}
    B {from=MaterialExpressionSaturate_10}
   ]
  }
  MaterialExpressionMultiply_90 {
   class=MaterialExpressionMultiply
   inputs [
    A {from=MaterialExpressionScalarParameter_72}
    B {from=MaterialExpressionMultiply_92}
   ]
  }
  MaterialExpressionMultiply_91 {
   class=MaterialExpressionMultiply
   inputs [
    A {from=MaterialExpressionMultiply_92}
    B {from=MaterialExpressionScalarParameter_73}
   ]
  }
  MaterialExpressionMultiply_92 {
   class=MaterialExpressionMultiply
   inputs [
    A {from=MaterialExpressionNamedRerouteDeclaration_20}
    B {from=MaterialExpressionScalarParameter_75}
   ]
  }
  MaterialExpressionNamedRerouteDeclaration_0 {
   class=MaterialExpressionNamedRerouteDeclaration
   inputs [None {from=MaterialExpressionTransform_1}]
   properties {name=AddedWPO, nodeColor=<rgba 0 0.2266 1 1>}
  }
  MaterialExpressionNamedRerouteDeclaration_1 {
   class=MaterialExpressionNamedRerouteDeclaration
   inputs [None {from=MaterialExpressionMaterialFunctionCall_2, output=Result}]
   properties {name=emissive, nodeColor=<rgba 0 1 0.7188 1>}
  }
  MaterialExpressionNamedRerouteDeclaration_10 {
   class=MaterialExpressionNamedRerouteDeclaration
   inputs [None {from=MaterialExpressionStaticSwitchParameter_2}]
   properties {name=SplineThicken, nodeColor=<rgba 1 0.6797 0 1>}
  }
  MaterialExpressionNamedRerouteDeclaration_11 {
   class=MaterialExpressionNamedRerouteDeclaration
   inputs [None {from=MaterialExpressionAdd_0}]
   properties {name=WPO, nodeColor=<rgba 0.0781 0 1 1>}
  }
  MaterialExpressionNamedRerouteDeclaration_15 {
   class=MaterialExpressionNamedRerouteDeclaration
   inputs [None {from=MaterialExpressionTransform_0}]
   properties {name=WPO Distortion, nodeColor=<rgba 1 0 0.9844 1>}
  }
  MaterialExpressionNamedRerouteDeclaration_16 {
   class=MaterialExpressionNamedRerouteDeclaration
   inputs [None {from=MaterialExpressionStaticSwitchParameter_3}]
   properties {name=Interpolate WPO, nodeColor=<rgba 0.3828 1 0 1>}
  }
  MaterialExpressionNamedRerouteDeclaration_2 {
   class=MaterialExpressionNamedRerouteDeclaration
   inputs [None {from=MaterialExpressionShadingPathSwitch_0}]
   properties {name=opacity, nodeColor=<rgba 0.4766 0 1 1>}
  }
  MaterialExpressionNamedRerouteDeclaration_20 {
   class=MaterialExpressionNamedRerouteDeclaration
   inputs [None {from=MaterialExpressionSaturate_1}]
   properties {name=EdgeGradient, nodeColor=<rgba 1 0 0.9141 1>}
  }
  MaterialExpressionNamedRerouteDeclaration_3 {
   class=MaterialExpressionNamedRerouteDeclaration
   inputs [None {from=MaterialExpressionLinearInterpolate_9}]
   properties {name=tex0, nodeColor=<rgba 0 1 0.1562 1>}
  }
  MaterialExpressionNamedRerouteDeclaration_4 {
   class=MaterialExpressionNamedRerouteDeclaration
   inputs [None {from=MaterialExpressionMultiply_5}]
   properties {name=PanSpeed, nodeColor=<rgba 1 0 0.7266 1>}
  }
  MaterialExpressionNamedRerouteUsage_0 {class=MaterialExpressionNamedRerouteUsage}
  MaterialExpressionNamedRerouteUsage_1 {class=MaterialExpressionNamedRerouteUsage}
  MaterialExpressionNamedRerouteUsage_10 {class=MaterialExpressionNamedRerouteUsage}
  MaterialExpressionNamedRerouteUsage_11 {class=MaterialExpressionNamedRerouteUsage}
  MaterialExpressionNamedRerouteUsage_12 {class=MaterialExpressionNamedRerouteUsage}
  MaterialExpressionNamedRerouteUsage_13 {class=MaterialExpressionNamedRerouteUsage}
  MaterialExpressionNamedRerouteUsage_14 {class=MaterialExpressionNamedRerouteUsage}
  MaterialExpressionNamedRerouteUsage_15 {class=MaterialExpressionNamedRerouteUsage}
  MaterialExpressionNamedRerouteUsage_16 {class=MaterialExpressionNamedRerouteUsage}
  MaterialExpressionNamedRerouteUsage_17 {class=MaterialExpressionNamedRerouteUsage}
  MaterialExpressionNamedRerouteUsage_18 {class=MaterialExpressionNamedRerouteUsage}
  MaterialExpressionNamedRerouteUsage_19 {class=MaterialExpressionNamedRerouteUsage}
  MaterialExpressionNamedRerouteUsage_2 {class=MaterialExpressionNamedRerouteUsage}
  MaterialExpressionNamedRerouteUsage_20 {class=MaterialExpressionNamedRerouteUsage}
  MaterialExpressionNamedRerouteUsage_21 {class=MaterialExpressionNamedRerouteUsage}
  MaterialExpressionNamedRerouteUsage_24 {class=MaterialExpressionNamedRerouteUsage}
  MaterialExpressionNamedRerouteUsage_25 {class=MaterialExpressionNamedRerouteUsage}
  MaterialExpressionNamedRerouteUsage_3 {class=MaterialExpressionNamedRerouteUsage}
  MaterialExpressionNamedRerouteUsage_4 {class=MaterialExpressionNamedRerouteUsage}
  MaterialExpressionNamedRerouteUsage_5 {class=MaterialExpressionNamedRerouteUsage}
  MaterialExpressionNamedRerouteUsage_6 {class=MaterialExpressionNamedRerouteUsage}
  MaterialExpressionNamedRerouteUsage_7 {class=MaterialExpressionNamedRerouteUsage}
  MaterialExpressionNamedRerouteUsage_8 {class=MaterialExpressionNamedRerouteUsage}
  MaterialExpressionNamedRerouteUsage_9 {class=MaterialExpressionNamedRerouteUsage}
  MaterialExpressionOneMinus_0 {class=MaterialExpressionOneMinus, inputs [None {from=MaterialExpressionComponentMask_0}]}
  MaterialExpressionOneMinus_1 {class=MaterialExpressionOneMinus, inputs [None {from=MaterialExpressionComponentMask_9}]}
  MaterialExpressionOneMinus_2 {class=MaterialExpressionOneMinus, inputs [None {from=MaterialExpressionComponentMask_13}]}
  MaterialExpressionOneMinus_3 {class=MaterialExpressionOneMinus, inputs [None {from=MaterialExpressionSphereMask_1}]}
  MaterialExpressionOneMinus_4 {class=MaterialExpressionOneMinus, inputs [None {from=MaterialExpressionMaterialFunctionCall_5, output=R}]}
  MaterialExpressionOneMinus_5 {class=MaterialExpressionOneMinus, inputs [None {from=MaterialExpressionTextureCoordinate_23}]}
  MaterialExpressionOneMinus_6 {class=MaterialExpressionOneMinus, inputs [None {from=MaterialExpressionComponentMask_18}]}
  MaterialExpressionPanner_0 {
   class=MaterialExpressionPanner
   inputs [
    Coordinate {from=MaterialExpressionAdd_5}
    Time {from=MaterialExpressionNamedRerouteUsage_13}
    Speed {from=""}
   ]
   properties {speedX=-0.3}
  }
  MaterialExpressionPanner_1 {
   class=MaterialExpressionPanner
   inputs [
    Coordinate {from=MaterialExpressionAdd_2}
    Time {from=MaterialExpressionNamedRerouteUsage_15}
    Speed {from=""}
   ]
   properties {speedX=-1}
  }
  MaterialExpressionPanner_10 {
   class=MaterialExpressionPanner
   inputs [
    Coordinate {from=MaterialExpressionMultiply_56}
    Time {from=""}
    Speed {from=MaterialExpressionAppendVector_37}
   ]
  }
  MaterialExpressionPanner_11 {
   class=MaterialExpressionPanner
   inputs [
    Coordinate {from=MaterialExpressionMultiply_74}
    Time {from=""}
    Speed {from=MaterialExpressionAppendVector_42}
   ]
  }
  MaterialExpressionPanner_2 {
   class=MaterialExpressionPanner
   inputs [
    Coordinate {from=MaterialExpressionAdd_6}
    Time {from=MaterialExpressionNamedRerouteUsage_19}
    Speed {from=""}
   ]
   properties {speedX=-0.5}
  }
  MaterialExpressionPanner_3 {
   class=MaterialExpressionPanner
   inputs [
    Coordinate {from=MaterialExpressionAdd_14}
    Time {from=MaterialExpressionNamedRerouteUsage_17}
    Speed {from=""}
   ]
   properties {speedX=-0.5}
  }
  MaterialExpressionPanner_4 {
   class=MaterialExpressionPanner
   inputs [
    Coordinate {from=MaterialExpressionAdd_9}
    Time {from=MaterialExpressionNamedRerouteUsage_20}
    Speed {from=""}
   ]
   properties {speedX=-0.3}
  }
  MaterialExpressionPanner_9 {
   class=MaterialExpressionPanner
   inputs [
    Coordinate {from=MaterialExpressionMultiply_52}
    Time {from=""}
    Speed {from=MaterialExpressionAppendVector_34}
   ]
  }
  MaterialExpressionParticleColor_0 {class=MaterialExpressionParticleColor, inputs [], properties {}}
  MaterialExpressionParticlePositionWS_0 {class=MaterialExpressionParticlePositionWS, inputs [], properties {}}
  MaterialExpressionPower_0 {
   class=MaterialExpressionPower
   inputs [
    Base {from=MaterialExpressionComponentMask_11}
    Exp {from=""}
   ]
   properties {constExponent=0.7}
  }
  MaterialExpressionPower_1 {
   class=MaterialExpressionPower
   inputs [
    Base {from=MaterialExpressionComponentMask_1}
    Exp {from=""}
   ]
   properties {constExponent=0.7}
  }
  MaterialExpressionPower_10 {
   class=MaterialExpressionPower
   inputs [
    Base {from=MaterialExpressionComponentMask_7}
    Exp {from=""}
   ]
   properties {constExponent=3}
  }
  MaterialExpressionPower_14 {
   class=MaterialExpressionPower
   inputs [
    Base {from=MaterialExpressionOneMinus_1}
    Exp {from=MaterialExpressionConstant_23}
   ]
   properties {constExponent=0.5}
  }
  MaterialExpressionPower_15 {
   class=MaterialExpressionPower
   inputs [
    Base {from=MaterialExpressionComponentMask_9}
    Exp {from=MaterialExpressionConstant_24}
   ]
   properties {constExponent=0.5}
  }
  MaterialExpressionPower_16 {
   class=MaterialExpressionPower
   inputs [
    Base {from=MaterialExpressionComponentMask_10}
    Exp {from=""}
   ]
   properties {constExponent=1}
  }
  MaterialExpressionPower_17 {
   class=MaterialExpressionPower
   inputs [
    Base {from=MaterialExpressionOneMinus_2}
    Exp {from=MaterialExpressionConstant_27}
   ]
   properties {constExponent=0.5}
  }
  MaterialExpressionPower_18 {
   class=MaterialExpressionPower
   inputs [
    Base {from=MaterialExpressionComponentMask_13}
    Exp {from=MaterialExpressionConstant_28}
   ]
   properties {constExponent=0.5}
  }
  MaterialExpressionPower_19 {
   class=MaterialExpressionPower
   inputs [
    Base {from=MaterialExpressionComponentMask_16}
    Exp {from=MaterialExpressionScalarParameter_50}
   ]
  }
  MaterialExpressionPower_2 {
   class=MaterialExpressionPower
   inputs [
    Base {from=MaterialExpressionChannelMaskParameter_0, output=R}
    Exp {from=MaterialExpressionScalarParameter_8}
   ]
   properties {constExponent=2.2}
  }
  MaterialExpressionPower_3 {
   class=MaterialExpressionPower
   inputs [
    Base {from=MaterialExpressionOneMinus_0}
    Exp {from=""}
   ]
   properties {constExponent=3}
  }
  MaterialExpressionPower_7 {
   class=MaterialExpressionPower
   inputs [
    Base {from=MaterialExpressionMultiply_11}
    Exp {from=MaterialExpressionScalarParameter_31}
   ]
   properties {constExponent=5}
  }
  MaterialExpressionPower_9 {
   class=MaterialExpressionPower
   inputs [
    Base {from=MaterialExpressionTextureSample_4, output=R}
    Exp {from=MaterialExpressionSaturate_9}
   ]
  }
  MaterialExpressionReroute_0 {class=MaterialExpressionReroute, inputs [None {from=MaterialExpressionPower_2}]}
  MaterialExpressionReroute_1 {class=MaterialExpressionReroute, inputs [None {from=MaterialExpressionTextureSampleParameter2D_1, output=B}]}
  MaterialExpressionReroute_11 {class=MaterialExpressionReroute, inputs [None {from=MaterialExpressionScalarParameter_59}]}
  MaterialExpressionReroute_12 {class=MaterialExpressionReroute, inputs [None {from=MaterialExpressionAppendVector_39}]}
  MaterialExpressionRound_0 {class=MaterialExpressionRound, inputs [None {from=MaterialExpressionPower_9, output=""}], properties {}}
  MaterialExpressionSaturate_0 {class=MaterialExpressionSaturate, inputs [None {from=MaterialExpressionMultiply_3}]}
  MaterialExpressionSaturate_1 {class=MaterialExpressionSaturate, inputs [None {from=MaterialExpressionMultiply_1}]}
  MaterialExpressionSaturate_10 {class=MaterialExpressionSaturate, inputs [None {from=MaterialExpressionLinearInterpolate_5}]}
  MaterialExpressionSaturate_11 {class=MaterialExpressionSaturate, inputs [None {from=MaterialExpressionMultiply_65}]}
  MaterialExpressionSaturate_2 {class=MaterialExpressionSaturate, inputs [None {from=MaterialExpressionSubtract_5}]}
  MaterialExpressionSaturate_3 {class=MaterialExpressionSaturate, inputs [None {from=MaterialExpressionLinearInterpolate_6}]}
  MaterialExpressionSaturate_4 {class=MaterialExpressionSaturate, inputs [None {from=MaterialExpressionMultiply_16}]}
  MaterialExpressionSaturate_9 {class=MaterialExpressionSaturate, inputs [None {from=MaterialExpressionLinearInterpolate_4}]}
  MaterialExpressionScalarParameter_0 {class=MaterialExpressionScalarParameter, properties {defaultValue=1, parameterName=Add Noise}}
  MaterialExpressionScalarParameter_1 {class=MaterialExpressionScalarParameter, properties {defaultValue=1, parameterName=Use Color Curve}}
  MaterialExpressionScalarParameter_10 {class=MaterialExpressionScalarParameter, properties {parameterName=UseSwizzle}}
  MaterialExpressionScalarParameter_11 {class=MaterialExpressionScalarParameter, properties {defaultValue=1, parameterName=SpeedMult}}
  MaterialExpressionScalarParameter_12 {class=MaterialExpressionScalarParameter, properties {defaultValue=0.2, parameterName=AngleFade}}
  MaterialExpressionScalarParameter_13 {class=MaterialExpressionScalarParameter, properties {defaultValue=3, parameterName=DepthFadeDist}}
  MaterialExpressionScalarParameter_14 {class=MaterialExpressionScalarParameter, properties {defaultValue=0.1, parameterName=Dark noise min}}
  MaterialExpressionScalarParameter_15 {class=MaterialExpressionScalarParameter, properties {parameterName=Fixed tips}}
  MaterialExpressionScalarParameter_16 {class=MaterialExpressionScalarParameter, properties {group=AddedWPO, parameterName=AddedWPO_Strength, sortPriority=2}}
  MaterialExpressionScalarParameter_17 {
   class=MaterialExpressionScalarParameter
   properties {defaultValue=1, group=AddedWPO, parameterName=AddedWPO_Gradient_PostMult, sortPriority=3}
  }
  MaterialExpressionScalarParameter_18 {
   class=MaterialExpressionScalarParameter
   properties {defaultValue=1, group=AddedWPO, parameterName=AddedWPO_Gradient_PreMult, sortPriority=3}
  }
  MaterialExpressionScalarParameter_19 {class=MaterialExpressionScalarParameter, properties {group=AddedWPO, parameterName=AddedWPO_Subtract}}
  MaterialExpressionScalarParameter_2 {class=MaterialExpressionScalarParameter, properties {parameterName=Dissolve mask Min}}
  MaterialExpressionScalarParameter_21 {class=MaterialExpressionScalarParameter, properties {parameterName=UseNoiseAsCurveTime}}
  MaterialExpressionScalarParameter_22 {class=MaterialExpressionScalarParameter, properties {parameterName=UseNoiseAsCurveTime}}
  MaterialExpressionScalarParameter_28 {class=MaterialExpressionScalarParameter, properties {defaultValue=1, group=SplineThicken, parameterName=Spline Width Mult}}
  MaterialExpressionScalarParameter_3 {class=MaterialExpressionScalarParameter, properties {defaultValue=1, parameterName=Dissolve mask Max}}
  MaterialExpressionScalarParameter_31 {class=MaterialExpressionScalarParameter, properties {defaultValue=2, group=AddedWPO, parameterName=AddedWPO_GradientPower, sortPriority=4}}
  MaterialExpressionScalarParameter_4 {class=MaterialExpressionScalarParameter, properties {defaultValue=-2, parameterName=Disappear min}}
  MaterialExpressionScalarParameter_43 {class=MaterialExpressionScalarParameter, properties {defaultValue=10, group=WPO, parameterName=WPO Z Noise Amount}}
  MaterialExpressionScalarParameter_44 {class=MaterialExpressionScalarParameter, properties {defaultValue=3, group=WPO, parameterName=WPO Noise Intensity}}
  MaterialExpressionScalarParameter_45 {class=MaterialExpressionScalarParameter, properties {group=SplineThicken, parameterName=WPO Top Mask}}
  MaterialExpressionScalarParameter_46 {class=MaterialExpressionScalarParameter, properties {group=WPO, parameterName=UseDynNoiseIntensity}}
  MaterialExpressionScalarParameter_47 {class=MaterialExpressionScalarParameter, properties {defaultValue=1, group=WPO, parameterName=WPO Y Noise Amount}}
  MaterialExpressionScalarParameter_48 {class=MaterialExpressionScalarParameter, properties {defaultValue=5, group=WPO, parameterName=WPO X Noise Amount}}
  MaterialExpressionScalarParameter_49 {class=MaterialExpressionScalarParameter, properties {defaultValue=30, group=Interpolate, parameterName=Interpolate Clamp XY}}
  MaterialExpressionScalarParameter_5 {class=MaterialExpressionScalarParameter, properties {defaultValue=1, parameterName=Disappear max}}
  MaterialExpressionScalarParameter_50 {class=MaterialExpressionScalarParameter, properties {defaultValue=2.5, group=Interpolate, parameterName=Interpolate Power}}
  MaterialExpressionScalarParameter_51 {class=MaterialExpressionScalarParameter, properties {defaultValue=1, group=Interpolate, parameterName=Interpolate Clamp Z}}
  MaterialExpressionScalarParameter_57 {class=MaterialExpressionScalarParameter, properties {group=WPO, parameterName=WPO Z Clamp}}
  MaterialExpressionScalarParameter_58 {class=MaterialExpressionScalarParameter, properties {parameterName=RandomWPO_Offset}}
  MaterialExpressionScalarParameter_59 {class=MaterialExpressionScalarParameter, properties {parameterName=RandomWPO_Offset}}
  MaterialExpressionScalarParameter_6 {class=MaterialExpressionScalarParameter, properties {defaultValue=1, parameterName=Lerp Center Dissolve}}
  MaterialExpressionScalarParameter_72 {class=MaterialExpressionScalarParameter, properties {defaultValue=2, group=SplineThicken, parameterName=SplineThickenWidth}}
  MaterialExpressionScalarParameter_73 {class=MaterialExpressionScalarParameter, properties {defaultValue=15, group=SplineThicken, parameterName=SplineThickenWidthTip}}
  MaterialExpressionScalarParameter_74 {class=MaterialExpressionScalarParameter, properties {defaultValue=5, group=SplineThicken, parameterName=ThickenCorePower}}
  MaterialExpressionScalarParameter_75 {class=MaterialExpressionScalarParameter, properties {defaultValue=1, group=SplineThicken, parameterName=ThickenCoreSize}}
  MaterialExpressionScalarParameter_76 {class=MaterialExpressionScalarParameter, properties {group=SplineThicken, parameterName=ThickenCore?}}
  MaterialExpressionScalarParameter_8 {class=MaterialExpressionScalarParameter, properties {defaultValue=1, group=Tex, parameterName=Tex alpha fix}}
  MaterialExpressionScalarParameter_9 {class=MaterialExpressionScalarParameter, properties {parameterName=Lerp Dy Noise}}
  MaterialExpressionShadingPathSwitch_0 {
   class=MaterialExpressionShadingPathSwitch
   inputs [
    Default {from=MaterialExpressionDepthFade_0, output=""}
    Deferred {from="", output=""}
    Forward {from="", output=""}
    Mobile {from=MaterialExpressionMultiply_28, output=""}
   ]
   properties {desc=Migrated from FeatureLevelSwitch}
  }
  MaterialExpressionSphereMask_1 {
   class=MaterialExpressionSphereMask
   inputs [
    A {from=MaterialExpressionNamedRerouteUsage_8, output=""}
    B {from=MaterialExpressionConstant2Vector_1, output=""}
    Radius {from="", output=""}
    Hardness {from="", output=""}
   ]
   properties {attenuationRadius=0.5, hardnessPercent=50}
  }
  MaterialExpressionStaticSwitchParameter_2 {
   class=MaterialExpressionStaticSwitchParameter
   inputs [
    True {from=MaterialExpressionMultiply_37, output=""}
    False {from=MaterialExpressionConstant_7, output=""}
   ]
   properties {defaultValue=true, group=SplineThicken, parameterName=UseSplineThicken}
  }
  MaterialExpressionStaticSwitchParameter_3 {
   class=MaterialExpressionStaticSwitchParameter
   inputs [
    True {from=MaterialExpressionLinearInterpolate_18, output=""}
    False {from=MaterialExpressionConstant_29, output=""}
   ]
   properties {defaultValue=true, group=Interpolate, parameterName=Use Interpolate WPO, sortPriority=0}
  }
  MaterialExpressionSubtract_0 {
   class=MaterialExpressionSubtract
   inputs [
    A {from=MaterialExpressionMaterialFunctionCall_16, output=Result}
    B {from=MaterialExpressionParticlePositionWS_0}
   ]
  }
  MaterialExpressionSubtract_1 {
   class=MaterialExpressionSubtract
   inputs [
    A {from=MaterialExpressionReroute_0}
    B {from=MaterialExpressionLinearInterpolate_10}
   ]
  }
  MaterialExpressionSubtract_5 {
   class=MaterialExpressionSubtract
   inputs [
    A {from=MaterialExpressionMultiply_40}
    B {from=MaterialExpressionScalarParameter_19}
   ]
  }
  MaterialExpressionTextureCoordinate_0 {class=MaterialExpressionTextureCoordinate}
  MaterialExpressionTextureCoordinate_18 {class=MaterialExpressionTextureCoordinate}
  MaterialExpressionTextureCoordinate_19 {class=MaterialExpressionTextureCoordinate}
  MaterialExpressionTextureCoordinate_2 {class=MaterialExpressionTextureCoordinate}
  MaterialExpressionTextureCoordinate_20 {class=MaterialExpressionTextureCoordinate}
  MaterialExpressionTextureCoordinate_22 {class=MaterialExpressionTextureCoordinate}
  MaterialExpressionTextureCoordinate_23 {class=MaterialExpressionTextureCoordinate}
  MaterialExpressionTextureObject_1 {
   class=MaterialExpressionTextureObject
   inputs []
   properties {samplerType=SAMPLERTYPE_LinearColor, texture=/Game/Global/Textures/Default/Tileable_Patterns/T_PerlinNoise_Uncompressed}
  }
  MaterialExpressionTextureSampleParameter2D_0 {
   class=MaterialExpressionTextureSampleParameter2D
   inputs [
    UVs {from=MaterialExpressionPanner_0, output=""}
    Apply View MipBias {from="", output=""}
   ]
   properties {parameterName=Tex_Default, texture=/Game/Effects/Niagara/MasterMaterials/SplineThickenTrail/Texture/T_FlameToon_Trail}
  }
  MaterialExpressionTextureSampleParameter2D_1 {
   class=MaterialExpressionTextureSampleParameter2D
   inputs [
    UVs {from=MaterialExpressionPanner_4, output=""}
    Apply View MipBias {from="", output=""}
   ]
   properties {parameterName=Tex_Default, texture=&T_FlameToon_Trail}
  }
  MaterialExpressionTextureSample_0 {
   class=MaterialExpressionTextureSample
   inputs [
    UVs {from=MaterialExpressionPanner_1}
    Tex {from=""}
    Apply View MipBias {from=""}
   ]
   properties {samplerType=SAMPLERTYPE_LinearGrayscale, texture=/Engine/Functions/Engine_MaterialFunctions02/ExampleContent/Textures/LowResBlurredNoise}
  }
  MaterialExpressionTextureSample_1 {
   class=MaterialExpressionTextureSample
   inputs [
    UVs {from=MaterialExpressionPanner_2}
    Tex {from=""}
    Apply View MipBias {from=""}
   ]
   properties {samplerType=SAMPLERTYPE_Grayscale, texture=/Game/Global/Textures/Default/Noise/LowResBlurredNoise}
  }
  MaterialExpressionTextureSample_2 {
   class=MaterialExpressionTextureSample
   inputs [
    UVs {from=MaterialExpressionAdd_31}
    Tex {from=""}
    Apply View MipBias {from=""}
   ]
   properties {samplerType=SAMPLERTYPE_LinearColor, texture=&T_PerlinNoise_Uncompressed}
  }
  MaterialExpressionTextureSample_4 {
   class=MaterialExpressionTextureSample
   inputs [
    UVs {from=MaterialExpressionPanner_3}
    Tex {from=""}
    Apply View MipBias {from=""}
   ]
   properties {samplerType=SAMPLERTYPE_LinearGrayscale, texture=/Game/Global/Textures/Default/Caustics/T_Nosie_Caustics_Grayscale}
  }
  MaterialExpressionTextureSample_5 {
   class=MaterialExpressionTextureSample
   inputs [
    UVs {from=MaterialExpressionAdd_32}
    Tex {from=MaterialExpressionTextureObject_1}
    Apply View MipBias {from=""}
   ]
   properties {samplerType=SAMPLERTYPE_LinearColor, texture=&T_PerlinNoise_Uncompressed}
  }
  MaterialExpressionTextureSample_6 {
   class=MaterialExpressionTextureSample
   inputs [
    UVs {from=MaterialExpressionAdd_33}
    Tex {from=MaterialExpressionTextureObject_1}
    Apply View MipBias {from=""}
   ]
   properties {samplerType=SAMPLERTYPE_LinearColor, texture=&T_PerlinNoise_Uncompressed}
  }
  MaterialExpressionTime_0 {class=MaterialExpressionTime}
  MaterialExpressionTransform_0 {class=MaterialExpressionTransform, inputs [None {from=MaterialExpressionAdd_22}], properties {transformSourceType=TRANSFORMSOURCE_Local}}
  MaterialExpressionTransform_1 {
   class=MaterialExpressionTransform
   inputs [None {from=MaterialExpressionMultiply_31}]
   properties {transformSourceType=TRANSFORMSOURCE_Local}
  }
  MaterialExpressionVectorParameter_0 {
   class=MaterialExpressionVectorParameter
   properties {group=AddedWPO, parameterName=AddeWPODirection, sortPriority=1, defaultValue=<rgba 0 0 1 0>}
  }
  MaterialExpressionVectorParameter_1 {class=MaterialExpressionVectorParameter, properties {parameterName=UV Scale, defaultValue=<rgba 0.8 1 0 0>}}
  MaterialExpressionVectorParameter_13 {
   class=MaterialExpressionVectorParameter
   properties {
    group=WPO
    parameterName=WPO Depth UV
    sortPriority=6
    channelNames {a=WPO Depth V Pan, b=WPO Depth U Pan, g=WPO Depth V, r=WPO Depth U}
    defaultValue=<rgba 0 0.6 0 0.2>
   }
  }
  MaterialExpressionVectorParameter_14 {
   class=MaterialExpressionVectorParameter
   properties {group=WPO, parameterName=WPO Noise UV, sortPriority=6, defaultValue=<rgba 0 0.65 0 0.1>}
  }
  MaterialExpressionVectorParameter_15 {
   class=MaterialExpressionVectorParameter
   properties {group=WPO, parameterName=WPO Noise UV 2, sortPriority=6, defaultValue=<rgba 0 0.5 0 0.2>}
  }
  MaterialExpressionVectorParameter_2 {class=MaterialExpressionVectorParameter, properties {parameterName="Color_Night (V3)", defaultValue=<rgba 1.5 1.5 1.5 1>}}
  MaterialExpressionVectorParameter_3 {class=MaterialExpressionVectorParameter, properties {parameterName=UV Scale, defaultValue=<rgba 0.8 1 0 0>}}
  MaterialExpressionVectorParameter_4 {class=MaterialExpressionVectorParameter, properties {parameterName=Dark Noise UV Scale, defaultValue=<rgba 1 1 0 0>}}
  MaterialExpressionVectorParameter_5 {class=MaterialExpressionVectorParameter, properties {parameterName=Bright Noise Scale, defaultValue=<rgba 1 1 0 0>}}
 }
 properties {
  bAutomaticallySetUsageInEditor=false
  bUsedWithNiagaraMeshParticles=true
  blendMode=BLEND_Translucent
  shadingModel=MSM_Unlit
  twoSided=true
  usedShadingModels=Unlit
 }
}
