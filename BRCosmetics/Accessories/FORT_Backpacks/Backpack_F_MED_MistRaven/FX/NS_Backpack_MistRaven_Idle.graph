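# Idle Niagara effect for the MistRaven back bling, layering a sprite airglow, flickering mesh flares, swaying smoke meshes and feathers shed from the skeletal mesh.
# User.Variant indexes arrays to pick colors and zero out spawn counts, switching between base and variant flare and smoke emitters.
# Front-end modifiers retune color and camera offset, and swap Flare_Variant for Flare_Variant_FE.
# Smoke feeds smoothed world position to Interpolate Dynamic as a lagging value.
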NiagaraSystem /BRCosmetics/Accessories/FORT_Backpacks/Backpack_F_MED_MistRaven/FX/NS_Backpack_MistRaven_Idle
dependencies {
 usedDataInterfaces [
  &NiagaraDataInterfaceArrayColor
  &NiagaraDataInterfaceCurve
  &NiagaraDataInterfaceArrayInt32
  &NiagaraDataInterfaceArrayFloat3
  &NiagaraDataInterfaceArrayFloat
  &NiagaraDataInterfaceSkeletalMesh
 ]
 usedDynamicInputs []
 usedModules [
  &SystemState
  &self:SystemSpawnScript.SystemScriptSource.SystemScriptGraph.NiagaraNodeAssignment_0.SetVariables_#1
  &self:SystemSpawnScript.SystemScriptSource.SystemScriptGraph.NiagaraNodeAssignment_1.SetVariables_#2
 ]
}
emitters {
 Airglow {
  emitterUpdateScript [
   &EmitterState(
    Inactive Response=::NewEnumerator1
    Life Cycle Mode=::NewEnumerator1
    Loop Behavior=::NewEnumerator1
    Loop Duration Mode=::NewEnumerator1
    UseLoopDelay=false
   )
   &SpawnBurst_Instantaneous(Use Spawn Probability=false, UseLoopCountLimit=false)
  ]
  particleSpawnScript [
   &InitializeParticle(
    Color=&SelectLinearColorFromArray(
     Array Sampling Mode=::NewEnumerator1
     Direct Array Index=&MakeIntFromFloat(Float=$User.Variant)
     Color Selection Array=DataInterface() {propertyValues {ColorData [<rgba 0.4792 0.0078 0.0565 1>, <rgba 0.4792 0.0001 0 0.4>]}}
     Random Seed=?
    )
    Lifetime=5
    Position Offset=$System.FlareOffset
    Sprite Size Mode=::NewEnumerator3
    Uniform Sprite Size=15
    UsePositionOffset=true
    Write Material Random=true
   )
   &CameraOffset(Camera Offset Amount=&FN_FrontEndFloatModifier(
     Gameplay=&SelectFloatFromArray(
      Direct Array Index=&MakeIntFromFloat(Float=$User.Variant)
      Float Selection Array=DataInterface() {propertyValues {FloatData [0,5]}}
     )
     Front End=0
    ))
  ]
  particleUpdateScript [
   &ParticleState(Kill Particles When Lifetime Has Elapsed=false, Let Infinitely Lived Particles Die When Emitter Deactivates=true)
   &DynamicMaterialParameters(
    Index 0 Param 1=3
    Index 0 Param 2=5
    Module.Index0Float0Write=true
    Module.Index0Float1Write=true
    Module.Index0Float2Write=true
    Module.Index0Float3Write=true
   )
   &ScaleColor(Scale Alpha=2, ScaleA=true, ScaleRGB=true)
   &SpriteFacingAndAlignment(Alignment Coordinate Space=::World, Sprite Alignment=<0 0 1>, Sprite Facing=<0 -1 0>, WriteAlignment=true, WriteFacing=true)
   /SetVariables(
    Particles.Facing2:Vector3f=<1 0 0>
    Particles.SpriteSize2=&Multiply_Vector2D(A=$Particles.SpriteSize, B=<0.5 1>)
   ) {isSetParameters=true}
   /SetVariables(Particles.Facing3:Vector3f=<0.001 0 1>, Particles.alignment3:Vector3f=<0 0 1>) {isSetParameters=true}
  ]
  rendererClasses [&NiagaraSpriteRendererProperties]
  renderers [
   {rendererClass=&NiagaraSpriteRendererProperties}
   {rendererClass=&NiagaraSpriteRendererProperties, rendererIndex=1}
   {rendererClass=&NiagaraSpriteRendererProperties, rendererIndex=2}
  ]
 }
 Feathers {
  emitterUpdateScript [
   &EmitterState(Life Cycle Mode=::NewEnumerator1, UseLoopDelay=false)
   &SpawnRate(Use Spawn Probability=false)
  ]
  particleSpawnScript [
   &InitializeParticle(
    Color=&SelectLinearColorFromArray(
     Array Sampling Mode=::NewEnumerator1
     Direct Array Index=&MakeIntFromFloat(Float=$User.Variant)
     Color Selection Array=DataInterface() {propertyValues {ColorData [<rgba 1 0.0162 0.1179 1>, <rgba 1 0.0484 0.0067 1>]}}
     Random Seed=?
    )
    Lifetime=&RandomRangeFloat(Minimum=0.5, Maximum=1.5)
    Mesh Scale Mode=::NewEnumerator1
    Mesh Uniform Scale Max=0.24
    Mesh Uniform Scale Min=0.12
    Sprite Size Mode=::NewEnumerator3
    UsePositionOffset=false
    Write Material Random=true
   )
   &SkeletalMeshLocation(
    MaskSpawnGroup=false
    Mesh Sampling Type=::NewEnumerator5
    OverrideRandomSeed=false
    Sampled Position Offset=&RandomRangeVector(Minimum=<-5 -5 -5>, Maximum=<5 5 5>)
    Skeletal Mesh=$System.Skeletal Mesh
   )
   &AddVelocity()
   &InitialMeshOrientation(Use Rotation Vector=false)
   &AddVelocity(Velocity=<0 0 -50>) {module=AddVelocity001}
   /SetVariables(Particles.VisibilityTag=&MakeIntFromFloat(Float=$User.Variant)) {isSetParameters=true}
  ]
  particleUpdateScript [
   &ParticleState()
   &ScaleMeshSize(Scale Factor=&VectorFromFloat(Value=&FloatFromCurve(FloatCurve=DataInterface() {propertyValues {Curve {
         defaultValue=1
         keys [
          {arriveTangent=-0.0047, interpMode=RCIM_Cubic, leaveTangent=-0.0047, tangentMode=RCTM_User, time=0, value=0}
          {arriveTangentWeight=0.0818, interpMode=RCIM_Cubic, leaveTangentWeight=0.3251, tangentMode=RCTM_User, time=0.2, value=1}
          {arriveTangentWeight=0.2835, interpMode=RCIM_Cubic, leaveTangentWeight=0.1521, tangentMode=RCTM_User, time=0.7, value=1}
          {arriveTangent=-0.0405, interpMode=RCIM_Cubic, leaveTangent=-0.0405, tangentMode=RCTM_User, time=1, value=0}
         ]
        }}})))
   &InheritSourceMovement(ApplyForce=false, ApplyPosition=true, ApplyPositionFalloff=false)
   &CurlNoiseForce(Bias Noise=false, Noise Frequency=25, Noise Strength=100, Pan Noise=false)
   &GravityForce(Gravity=<0 0 -175>)
   &MeshRotationForce(Rotation=&VectorFromFloat(Value=&Multiply_Float(
      A=&UniformAOrBFloat(
       A=&RandomRangeFloat(Minimum=-0.5, Maximum=-1)
       B=&RandomRangeFloat(Minimum=0.5, Maximum=1)
       Evaluation Type=::NewEnumerator0
       Recalculate Random Each Loop=true
       Override Seed=false
      )
      B=3
     )))
   &Drag(Use Linear Drag=true, Use Rotational Drag=false)
   &SolveForcesAndVelocity(Clamp Velocity=false, Limit Acceleration=false)
   &ScaleColor(Scale RGB=&VectorFromFloat(Value=0.8), ScaleA=true, ScaleRGB=true)
  ]
  rendererClasses [&NiagaraMeshRendererProperties]
  renderers [
   {rendererClass=&NiagaraMeshRendererProperties, rendererIndex=0}
   {rendererClass=&NiagaraMeshRendererProperties, rendererIndex=1}
  ]
 }
 Flare {
  emitterUpdateScript [
   &EmitterState(
    Inactive Response=::NewEnumerator1
    Life Cycle Mode=::NewEnumerator1
    Loop Behavior=::NewEnumerator1
    Loop Duration Mode=::NewEnumerator1
    UseLoopDelay=false
   )
   &SpawnBurst_Instantaneous(
    Spawn Count=&SelectIntFromArray(
     Direct Array Index=&MakeIntFromFloat(Float=$User.Variant)
     Int Selection Array=DataInterface() {propertyValues {IntData [4,0]}}
    )
    Use Spawn Probability=false
    UseLoopCountLimit=false
   )
  ]
  particleSpawnScript [
   &InitializeParticle(
    Color=&MultiplyLinearColor(
     Linear Color=<rgba 1 1 1 1>
     Scale Factor=&MakeVector4FromVectorAndFloat(
      "Vector (RGB)"=&FN_FrontEndVector3Modifier(Gameplay=<1 1 1>, Front End=<1 1 0.65>)
      "Float (Alpha)"=1
     )
    )
    Lifetime=5
    Mesh Scale=&Multiply_VectorByFloat(
     Vector=&SelectVectorFromArray(
      Direct Array Index=&ReturnExecIndex()
      Vector Selection Array=DataInterface() {propertyValues {FloatData [<1 1 1>, <1 1 2>, <1 1 1>, <1 1 2>]}}
     )
     Float=1.5
    )
    Mesh Scale Mode=::NewEnumerator4
    Position Offset=$System.FlareOffset
    Sprite Size Mode=::NewEnumerator3
    Uniform Sprite Size=20
    UsePositionOffset=true
    Write Material Random=true
   )
   &CameraOffset(Camera Offset Amount=2)
   &InitialMeshOrientation(Mesh Axis To Orient=<0 1 0>, Mesh Orientation Mode=::NewEnumerator2, Orientation Vector=<0 0 1>, Use Rotation Vector=false)
  ]
  particleUpdateScript [
   &ParticleState(Kill Particles When Lifetime Has Elapsed=false, Let Infinitely Lived Particles Die When Emitter Deactivates=true)
   &InterpolateOverTime(
    Rate Of Change=4
    Target Float Value:NiagaraFloat=&RandomRangeFloat(Minimum=0.6, Maximum=1.4, Evaluation Type=::NewEnumerator1)
   ) {module=InterpolateOverTime002}
   &ScaleMeshSize(Scale Factor=&VectorFromFloat(Value=$Particles.InterpolateOverTime002.InterpolateOverTime_Float.Moving Average))
   &ScaleSpriteSize(Scale Sprite Size Mode=::NewEnumerator0, Uniform Scale Factor=$Particles.InterpolateOverTime002.InterpolateOverTime_Float.Moving Average)
   &SkeletalMeshLocation(Bone Sampling Mode=::NewEnumerator4, MaskSpawnGroup=false, Sampled Position Offset=<0 0 3>, Skeletal Mesh=$System.Skeletal Mesh) {enabled=false}
   &InterpolateOverTime(
    Rate Of Change=2
    Target Float Value:NiagaraFloat=&RandomRangeFloat(Minimum=1, Maximum=1.35, Evaluation Type=::NewEnumerator1)
   ) {module=InterpolateOverTime003}
   &FN_Sprite_Attributes(
    Camera Offset:NiagaraFloat=&Add_Float(A=$Particles.CameraOffset, B=2)
    Color=<rgba 0.05 0.2 0.5 1>
    Create Camera Offset Attribute=true
    Create Color Attribute=true
    Create SpriteSize Attribute=true
    SpriteSize=&Multiply_Vector2D_ByFloat(Float=$Particles.InterpolateOverTime003.InterpolateOverTime_Float.Moving Average, Vector2D=$Particles.SpriteSize)
    Write to Module Namespace=true
   )
   /SetVariables(
    Particles.Sort:NiagaraInt32=&ReturnExecIndex()
    Particles.VisibilityTag=&ReturnExecIndex()
   ) {isSetParameters=true, module=SetVariables_#3}
   &FN_Mesh_Attributes(
    Create Scale Attribute=true
    Scale=&Multiply_VectorByFloat(Vector=&Multiply_Vector(A=$Particles.Scale, B=<1 1 1>), Float=0.8)
    Write to Module Namespace=true
   )
   /SetVariables(Particles.CameraOffset_02=6) {isSetParameters=true, module=SetVariables_#4}
   /SetVariables(Particles.DarkColor:LinearColor=&MakeLinearColorFromVectorAndFloat(
     "Vector (RGB)"=&FN_FrontEndVector3Modifier(Gameplay=<1 1 1>, Front End=<0.25 0.25 0.25>)
     "Float (Alpha)"=1
    )) {isSetParameters=true}
  ]
  rendererClasses [&NiagaraMeshRendererProperties]
  renderers [
   {rendererClass=&NiagaraMeshRendererProperties, rendererIndex=0}
   {rendererClass=&NiagaraMeshRendererProperties, rendererIndex=1}
   {rendererClass=&NiagaraMeshRendererProperties, rendererIndex=2}
   {rendererClass=&NiagaraMeshRendererProperties, rendererIndex=3}
   {rendererClass=&NiagaraMeshRendererProperties, rendererIndex=4}
   {rendererClass=&NiagaraMeshRendererProperties, rendererIndex=5}
  ]
 }
 Flare_Variant {
  emitterUpdateScript [
   &EmitterState(
    Inactive Response=::NewEnumerator1
    Life Cycle Mode=::NewEnumerator1
    Loop Behavior=::NewEnumerator1
    Loop Duration Mode=::NewEnumerator1
    UseLoopDelay=false
   )
   &SpawnBurst_Instantaneous(
    Spawn Count=&Multiply_Int(
     A=&SelectIntFromArray(
      Direct Array Index=&MakeIntFromFloat(Float=$User.Variant)
      Int Selection Array=DataInterface() {propertyValues {IntData [0,4]}}
     )
     B=&MakeIntFromFloat(Float=&FN_FrontEndFloatModifier(Gameplay=1, Front End=0))
    )
    Use Spawn Probability=false
    UseLoopCountLimit=false
   )
  ]
  particleSpawnScript [
   &InitializeParticle(
    Color=&MultiplyLinearColor(
     Linear Color=<rgba 1 1 1 1>
     Scale Factor=&MakeVector4FromVectorAndFloat(
      "Vector (RGB)"=&FN_FrontEndVector3Modifier(Gameplay=<1 1 1>, Front End=<0.8 1 1>)
      "Float (Alpha)"=1
     )
    )
    Lifetime=5
    Mesh Scale=&Multiply_VectorByFloat(
     Vector=&SelectVectorFromArray(
      Direct Array Index=&ReturnExecIndex()
      Vector Selection Array=DataInterface() {propertyValues {FloatData [<1 1 1>, <1 1 2>, <1 1 1>, <1 1 2>]}}
     )
     Float=1.5
    )
    Mesh Scale Mode=::NewEnumerator4
    Position Offset=$System.FlareOffset
    Sprite Size Mode=::NewEnumerator3
    Uniform Sprite Size=20
    UsePositionOffset=true
    Write Material Random=true
   )
   &CameraOffset(Camera Offset Amount=2)
   &InitialMeshOrientation(Mesh Axis To Orient=<0 1 0>, Mesh Orientation Mode=::NewEnumerator2, Orientation Vector=<0 0 1>, Use Rotation Vector=false)
  ]
  particleUpdateScript [
   &ParticleState(Kill Particles When Lifetime Has Elapsed=false, Let Infinitely Lived Particles Die When Emitter Deactivates=true)
   &InterpolateOverTime(
    Rate Of Change=4
    Target Float Value:NiagaraFloat=&RandomRangeFloat(Minimum=0.6, Maximum=1.4, Evaluation Type=::NewEnumerator1)
   ) {module=InterpolateOverTime002}
   &ScaleMeshSize(Scale Factor=&VectorFromFloat(Value=$Particles.InterpolateOverTime002.InterpolateOverTime_Float.Moving Average))
   &ScaleSpriteSize(Scale Sprite Size Mode=::NewEnumerator0, Uniform Scale Factor=$Particles.InterpolateOverTime002.InterpolateOverTime_Float.Moving Average)
   &SkeletalMeshLocation(Bone Sampling Mode=::NewEnumerator4, MaskSpawnGroup=false, Sampled Position Offset=<0 0 3>, Skeletal Mesh=$System.Skeletal Mesh) {enabled=false}
   &InterpolateOverTime(
    Rate Of Change=2
    Target Float Value:NiagaraFloat=&RandomRangeFloat(Minimum=1, Maximum=1.35, Evaluation Type=::NewEnumerator1)
   ) {module=InterpolateOverTime003}
   &FN_Sprite_Attributes(
    Camera Offset:NiagaraFloat=&Add_Float(A=$Particles.CameraOffset, B=2)
    Color=<rgba 0.05 0.2 0.5 1>
    Create Camera Offset Attribute=true
    Create Color Attribute=true
    Create SpriteSize Attribute=true
    SpriteSize=&Multiply_Vector2D_ByFloat(Float=$Particles.InterpolateOverTime003.InterpolateOverTime_Float.Moving Average, Vector2D=$Particles.SpriteSize)
    Write to Module Namespace=true
   )
   /SetVariables(
    Particles.Sort:NiagaraInt32=&ReturnExecIndex()
    Particles.VisibilityTag=&ReturnExecIndex()
   ) {isSetParameters=true, module=SetVariables_#3}
   &FN_Mesh_Attributes(
    Create Scale Attribute=true
    Scale=&Multiply_VectorByFloat(Vector=&Multiply_Vector(A=$Particles.Scale, B=<1 1 1>), Float=0.8)
    Write to Module Namespace=true
   )
   /SetVariables(Particles.CameraOffset_02=6) {isSetParameters=true, module=SetVariables_#4}
   /SetVariables(Particles.DarkColor:LinearColor=&MakeLinearColorFromVectorAndFloat(
     "Vector (RGB)"=&FN_FrontEndVector3Modifier(Gameplay=<1 1 1>, Front End=<0.35 0.35 0.35>)
     "Float (Alpha)"=1
    )) {isSetParameters=true, module=SetVariables_#5}
  ]
  rendererClasses [&NiagaraMeshRendererProperties]
  renderers [
   {rendererClass=&NiagaraMeshRendererProperties, rendererIndex=0}
   {rendererClass=&NiagaraMeshRendererProperties, rendererIndex=1}
   {rendererClass=&NiagaraMeshRendererProperties, rendererIndex=2}
   {rendererClass=&NiagaraMeshRendererProperties, rendererIndex=3}
   {rendererClass=&NiagaraMeshRendererProperties, rendererIndex=4}
   {rendererClass=&NiagaraMeshRendererProperties, rendererIndex=5}
  ]
 }
 Flare_Variant_FE {
  emitterUpdateScript [
   &EmitterState(
    Inactive Response=::NewEnumerator1
    Life Cycle Mode=::NewEnumerator1
    Loop Behavior=::NewEnumerator1
    Loop Duration Mode=::NewEnumerator1
    UseLoopDelay=false
   )
   &SpawnBurst_Instantaneous(
    Spawn Count=&Multiply_Int(
     A=&SelectIntFromArray(
      Direct Array Index=&MakeIntFromFloat(Float=$User.Variant)
      Int Selection Array=DataInterface() {propertyValues {IntData [0,4]}}
     )
     B=&MakeIntFromFloat(Float=&FN_FrontEndFloatModifier(Gameplay=0, Front End=1))
    )
    Use Spawn Probability=false
    UseLoopCountLimit=false
   )
  ]
  particleSpawnScript [
   &InitializeParticle(
    Color=&MultiplyLinearColor(
     Linear Color=<rgba 1 1 1 1>
     Scale Factor=&MakeVector4FromVectorAndFloat(
      "Vector (RGB)"=&FN_FrontEndVector3Modifier(Gameplay=<1 1 1>, Front End=<1.5 0.5 0.5>)
      "Float (Alpha)"=1
     )
    )
    Lifetime=5
    Mesh Scale=&Multiply_VectorByFloat(
     Vector=&SelectVectorFromArray(
      Direct Array Index=&ReturnExecIndex()
      Vector Selection Array=DataInterface() {propertyValues {FloatData [<1 1 1>, <1 1 2>, <1 1 1>, <1 1 2>]}}
     )
     Float=1.5
    )
    Mesh Scale Mode=::NewEnumerator4
    Position Offset=$System.FlareOffset
    Sprite Size Mode=::NewEnumerator3
    Uniform Sprite Size=20
    UsePositionOffset=true
    Write Material Random=true
   )
   &CameraOffset(Camera Offset Amount=2)
   &InitialMeshOrientation(Mesh Axis To Orient=<0 1 0>, Mesh Orientation Mode=::NewEnumerator2, Orientation Vector=<0 0 1>, Use Rotation Vector=false)
  ]
  particleUpdateScript [
   &ParticleState(Kill Particles When Lifetime Has Elapsed=false, Let Infinitely Lived Particles Die When Emitter Deactivates=true)
   &InterpolateOverTime(
    Rate Of Change=4
    Target Float Value:NiagaraFloat=&RandomRangeFloat(Minimum=0.6, Maximum=1.4, Evaluation Type=::NewEnumerator1)
   ) {module=InterpolateOverTime002}
   &ScaleMeshSize(Scale Factor=&VectorFromFloat(Value=$Particles.InterpolateOverTime002.InterpolateOverTime_Float.Moving Average))
   &ScaleSpriteSize(Scale Sprite Size Mode=::NewEnumerator0, Uniform Scale Factor=$Particles.InterpolateOverTime002.InterpolateOverTime_Float.Moving Average)
   &SkeletalMeshLocation(Bone Sampling Mode=::NewEnumerator4, MaskSpawnGroup=false, Sampled Position Offset=<0 0 3>, Skeletal Mesh=$System.Skeletal Mesh) {enabled=false}
   &InterpolateOverTime(
    Rate Of Change=2
    Target Float Value:NiagaraFloat=&RandomRangeFloat(Minimum=1, Maximum=1.35, Evaluation Type=::NewEnumerator1)
   ) {module=InterpolateOverTime003}
   &FN_Sprite_Attributes(
    Camera Offset:NiagaraFloat=&Add_Float(A=$Particles.CameraOffset, B=2)
    Color=<rgba 0.05 0.2 0.5 1>
    Create Camera Offset Attribute=true
    Create Color Attribute=true
    Create SpriteSize Attribute=true
    SpriteSize=&Multiply_Vector2D_ByFloat(Float=$Particles.InterpolateOverTime003.InterpolateOverTime_Float.Moving Average, Vector2D=$Particles.SpriteSize)
    Write to Module Namespace=true
   )
   /SetVariables(
    Particles.Sort:NiagaraInt32=&ReturnExecIndex()
    Particles.VisibilityTag=&ReturnExecIndex()
   ) {isSetParameters=true, module=SetVariables_#3}
   &FN_Mesh_Attributes(
    Create Scale Attribute=true
    Scale=&Multiply_VectorByFloat(Vector=&Multiply_Vector(A=$Particles.Scale, B=<1 1 1>), Float=0.8)
    Write to Module Namespace=true
   )
   /SetVariables(Particles.CameraOffset_02=6) {isSetParameters=true, module=SetVariables_#4}
   /SetVariables(Particles.DarkColor:LinearColor=&MakeLinearColorFromVectorAndFloat(
     "Vector (RGB)"=&FN_FrontEndVector3Modifier(Gameplay=<1 1 1>, Front End=<0.3 0.3 0.3>)
     "Float (Alpha)"=1
    )) {isSetParameters=true, module=SetVariables_#5}
  ]
  rendererClasses [&NiagaraMeshRendererProperties]
  renderers [
   {rendererClass=&NiagaraMeshRendererProperties, rendererIndex=0}
   {rendererClass=&NiagaraMeshRendererProperties, rendererIndex=1}
   {rendererClass=&NiagaraMeshRendererProperties, rendererIndex=2}
   {rendererClass=&NiagaraMeshRendererProperties, rendererIndex=3}
   {rendererClass=&NiagaraMeshRendererProperties, rendererIndex=4}
   {rendererClass=&NiagaraMeshRendererProperties, rendererIndex=5}
  ]
 }
 Smoke001 {
  emitterUpdateScript [
   &EmitterState(
    Inactive Response=::NewEnumerator1
    Life Cycle Mode=::NewEnumerator1
    Loop Behavior=::NewEnumerator1
    Loop Duration Mode=::NewEnumerator1
    Scalability Mode=::NewEnumerator1
    UseLoopDelay=false
   )
   &SpawnBurst_Instantaneous(
    Spawn Count=&SelectIntFromArray(
     Direct Array Index=&MakeIntFromFloat(Float=$User.Variant)
     Int Selection Array=DataInterface() {propertyValues {IntData [1,0]}}
    )
    Use Spawn Probability=false
    UseLoopCountLimit=false
   )
  ]
  particleSpawnScript [
   &InitializeParticle(
    Color=<rgba 1 0.2 0.35 1>
    Lifetime=5
    Mesh Scale=&Multiply_VectorByFloat(Vector=<1 0.35 1>, Float=0.8)
    Mesh Scale Mode=::NewEnumerator4
    Position Offset=&Add_Vector(A=$System.FlareOffset, B=<0 -0.5 0>)
    UsePositionOffset=true
    Write Material Random=true
   )
   &InitialMeshOrientation(Mesh Orientation Mode=::NewEnumerator2, Use Rotation Vector=false)
   &CameraOffset(Camera Offset Amount=4)
   /SetVariables(Particles.Color1:LinearColor=&Multiply_LinearColorByFloat(
     Linear Color=&MultiplyLinearColor(
      Linear Color=<rgba 1 1 1 1>
      Scale Factor=&MakeVector4FromVectorAndFloat(
       "Vector (RGB)"=&FN_FrontEndVector3Modifier(Gameplay=<1 1 1>, Front End=<0.25 0.25 0.25>)
       "Float (Alpha)"=1
      )
     )
     Float=1
    )) {isSetParameters=true, module=SetVariables_#6}
   /SetVariables(Particles.Color2=&Multiply_LinearColorByFloat(Linear Color=<rgba 1 1 1 3>, Float=1)) {isSetParameters=true, module=SetVariables_#7}
   /SetVariables(Particles.Color3:LinearColor=&Multiply_LinearColorByFloat(Linear Color=<rgba 1 1 1 3>, Float=0.3)) {isSetParameters=true, module=SetVariables_#8}
  ]
  particleUpdateScript [
   &ParticleState(Kill Particles When Lifetime Has Elapsed=false)
   /SetVariables(Particles.MeshOrientation=&MakeQuatFromAxisAngle(
     Axis=<0 1 0>
     AngleInDegrees=&Multiply_Float(
      A=&Waveform(
       Waveform Count=::NewEnumerator1
       "Waveform [1]"=::NewEnumerator0
       "[1] Frequency"=0.2
       "[1] Amplitude Min/Max"=<-1 1>
       "Waveform [2] Blend"=::NewEnumerator2
       "Waveform [2]"=::NewEnumerator1
       "[2] Input Position"=0
       "[2] Phase"=$Emitter.Age
       "[2] Phase Scale"=1
       "[2] Compound Frequency"=2
       "[2] Frequency"=0.1
       "[2] Amplitude Min/Max"=<0.5 1>
       "[2] Scale"=1
       "[2] Bias"=0
       "[2] Exponent"=1
       "[2] Pulse Duration"=0.1
       "[2] Clip Triangle Min/Max"=<0 0>
       "[2] Random Seed"=0
      )
      B=30
     )
    )) {isSetParameters=true, module=SetVariables_#9}
   &InterpolateOverTime(
    Rate Of Change=6
    Target Vector Value:Vector3f=&ConvertPositionToVector(
     Input Position=&TransformPosition(Position=$Particles.Position, Source Space=::Local, Destination Space=::World)
     Convert To Absolute Large World Vector=false
    )
    Variable Type=::NewEnumerator2
   )
   /SetVariables(Particles.Interpolate Dynamic=&MakeFloat4FromVectorAndFloat("Vector (RGB)"=$StackContext.InterpolateOverTime.InterpolateOverTime_V3_.Moving Average, "Float (Alpha)"=0.3)) {isSetParameters=true, module=SetVariables_#10}
   &ScaleColor(ScaleA=true, ScaleRGB=false)
   &DynamicMaterialParameters(
    Index 0 Param 2=$Engine.Time
    Index 0 Param 4=&Frac_Float(Float=$System.FN_Integrator002.Value)
    Module.Index0Param0Write=false
    Module.Index0Param1Write=false
    Module.Index0Param2Write=false
    Module.Index0Param3Write=true
   )
  ]
  rendererClasses [&NiagaraMeshRendererProperties]
  renderers [
   {rendererClass=&NiagaraMeshRendererProperties, rendererIndex=0}
   {rendererClass=&NiagaraMeshRendererProperties, rendererIndex=1}
   {rendererClass=&NiagaraMeshRendererProperties, rendererIndex=2}
   {rendererClass=&NiagaraMeshRendererProperties, rendererIndex=3}
  ]
 }
 SmokeVariant {
  emitterUpdateScript [
   &EmitterState(
    Inactive Response=::NewEnumerator1
    Life Cycle Mode=::NewEnumerator1
    Loop Behavior=::NewEnumerator1
    Loop Duration Mode=::NewEnumerator1
    Scalability Mode=::NewEnumerator1
    UseLoopDelay=false
   )
   &SpawnBurst_Instantaneous(
    Spawn Count=&SelectIntFromArray(
     Direct Array Index=&MakeIntFromFloat(Float=$User.Variant)
     Int Selection Array=DataInterface() {propertyValues {IntData [0,1]}}
    )
    Use Spawn Probability=false
    UseLoopCountLimit=false
   )
  ]
  particleSpawnScript [
   &InitializeParticle(
    Color=&SelectLinearColorFromArray(
     Array Sampling Mode=::NewEnumerator1
     Direct Array Index=&MakeIntFromFloat(Float=$User.Variant)
     Color Selection Array=DataInterface() {propertyValues {ColorData [<rgba 1 0.2 0.35 1>, <rgba 0.2188 0.2188 0.2188 2>]}}
     Random Seed=?
    )
    Lifetime=5
    Mesh Scale=&Multiply_VectorByFloat(Vector=<1 0.35 1>, Float=0.8)
    Mesh Scale Mode=::NewEnumerator4
    Position Offset=&Add_Vector(A=$System.FlareOffset, B=<0 -0.5 0>)
    UsePositionOffset=true
    Write Material Random=true
   )
   &InitialMeshOrientation(Mesh Orientation Mode=::NewEnumerator2, Use Rotation Vector=false)
   &CameraOffset(Camera Offset Amount=4)
   /SetVariables(Particles.Color1:LinearColor=&Multiply_LinearColorByFloat(
     Linear Color=&MultiplyLinearColor(
      Linear Color=<rgba 1 1 1 1>
      Scale Factor=&MakeVector4FromVectorAndFloat(
       "Vector (RGB)"=&FN_FrontEndVector3Modifier(Gameplay=<1 1 1>, Front End=<0.5 1 8>)
       "Float (Alpha)"=1
      )
     )
     Float=&FN_FrontEndFloatModifier(
      Gameplay=&SelectFloatFromArray(
       Direct Array Index=&MakeIntFromFloat(Float=$User.Variant)
       Float Selection Array=DataInterface() {propertyValues {FloatData [1,0.4]}}
      )
      Front End=&SelectFloatFromArray(
       Direct Array Index=&MakeIntFromFloat(Float=$User.Variant)
       Float Selection Array=DataInterface() {propertyValues {FloatData [1,1]}}
      )
     )
    )) {isSetParameters=true, module=SetVariables_#6}
   /SetVariables(Particles.Color2=&Multiply_LinearColorByFloat(
     Linear Color=&MultiplyLinearColor(
      Linear Color=<rgba 1 1 1 1>
      Scale Factor=&MakeVector4FromVectorAndFloat(
       "Vector (RGB)"=&FN_FrontEndVector3Modifier(Gameplay=<1 1 1>, Front End=<0.5 1 8>)
       "Float (Alpha)"=1
      )
     )
     Float=&FN_FrontEndFloatModifier(
      Gameplay=1
      Front End=&SelectFloatFromArray(
       Direct Array Index=&MakeIntFromFloat(Float=$User.Variant)
       Float Selection Array=DataInterface() {propertyValues {FloatData [1,2]}}
      )
     )
    )) {isSetParameters=true, module=SetVariables_#7}
   /SetVariables(Particles.Color3:LinearColor=&Multiply_LinearColorByFloat(
     Linear Color=<rgba 0.2448 0.2448 0.2448 3>
     Float=&SelectFloatFromArray(
      Direct Array Index=&MakeIntFromFloat(Float=$User.Variant)
      Float Selection Array=DataInterface() {propertyValues {FloatData [0.3,1]}}
     )
    )) {isSetParameters=true, module=SetVariables_#8}
  ]
  particleUpdateScript [
   &ParticleState(Kill Particles When Lifetime Has Elapsed=false)
   /SetVariables(Particles.MeshOrientation=&MakeQuatFromAxisAngle(
     Axis=<0 1 0>
     AngleInDegrees=&Multiply_Float(
      A=&Waveform(
       Waveform Count=::NewEnumerator1
       "Waveform [1]"=::NewEnumerator0
       "[1] Frequency"=0.2
       "[1] Amplitude Min/Max"=<-1 1>
       "Waveform [2] Blend"=::NewEnumerator2
       "Waveform [2]"=::NewEnumerator1
       "[2] Input Position"=0
       "[2] Phase"=$Emitter.Age
       "[2] Phase Scale"=1
       "[2] Compound Frequency"=2
       "[2] Frequency"=0.1
       "[2] Amplitude Min/Max"=<0.5 1>
       "[2] Scale"=1
       "[2] Bias"=0
       "[2] Exponent"=1
       "[2] Pulse Duration"=0.1
       "[2] Clip Triangle Min/Max"=<0 0>
       "[2] Random Seed"=0
      )
      B=30
     )
    )) {isSetParameters=true, module=SetVariables_#9}
   &InterpolateOverTime(
    Rate Of Change=6
    Target Vector Value:Vector3f=&ConvertPositionToVector(
     Input Position=&TransformPosition(Position=$Particles.Position, Source Space=::Local, Destination Space=::World)
     Convert To Absolute Large World Vector=false
    )
    Variable Type=::NewEnumerator2
   )
   /SetVariables(Particles.Interpolate Dynamic=&MakeFloat4FromVectorAndFloat("Vector (RGB)"=$StackContext.InterpolateOverTime.InterpolateOverTime_V3_.Moving Average, "Float (Alpha)"=0.3)) {isSetParameters=true, module=SetVariables_#10}
   &ScaleColor(
    Scale RGB=&FN_FrontEndVector3Modifier(Gameplay=<1 1 1>, Front End=<0.65 0.65 0.65>)
    ScaleA=true
    ScaleRGB=true
   )
   &DynamicMaterialParameters(
    Index 0 Param 2=$Engine.Time
    Index 0 Param 4=&Frac_Float(Float=$System.FN_Integrator002.Value)
    Module.Index0Param0Write=false
    Module.Index0Param1Write=false
    Module.Index0Param2Write=false
    Module.Index0Param3Write=true
   )
  ]
  rendererClasses [&NiagaraMeshRendererProperties]
  renderers [
   {rendererClass=&NiagaraMeshRendererProperties, rendererIndex=0}
   {rendererClass=&NiagaraMeshRendererProperties, rendererIndex=1}
   {rendererClass=&NiagaraMeshRendererProperties, rendererIndex=2}
   {rendererClass=&NiagaraMeshRendererProperties, rendererIndex=3}
  ]
 }
}
summary {
 systemName=NS_Backpack_MistRaven_Idle
 emitters [
  {emitterName=Feathers, rendererClasses [&NiagaraMeshRendererProperties]}
  {emitterName=Flare, rendererClasses [&NiagaraMeshRendererProperties]}
  {emitterName=Airglow, rendererClasses [&NiagaraSpriteRendererProperties]}
  {emitterName=Smoke001, rendererClasses [&NiagaraMeshRendererProperties]}
  {emitterName=Flare_Variant, rendererClasses [&NiagaraMeshRendererProperties]}
  {emitterName=SmokeVariant, rendererClasses [&NiagaraMeshRendererProperties]}
  {emitterName=Flare_Variant_FE, rendererClasses [&NiagaraMeshRendererProperties]}
 ]
 userVariables [
  User.bIsFrontEnd {type=NiagaraBool}
  User.bIsFrontEndPreview {type=NiagaraBool}
  User.Variant {type=NiagaraFloat, defaultValue=0, description=""}
 ]
}
