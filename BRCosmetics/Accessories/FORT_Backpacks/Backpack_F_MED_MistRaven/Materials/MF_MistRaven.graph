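# Material function for the MistRaven backpack cosmetic that adds animated colored smoke and burning edges to incoming material attributes, overriding base color and emissive.
# Panned noise and mask textures, distorted by a swirl flowmap, build named-reroute layers for base, detail, bright and black smoke plus burn edges.
# A local-position wing mask supports a time offset, and sine waves add flicker.
# Emissive passes through FX adjustment and time-of-day scaling.
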MaterialFunction /BRCosmetics/Accessories/FORT_Backpacks/Backpack_F_MED_MistRaven/Materials/MF_MistRaven
graph {nodes [
  n1 = GetMaterialAttributes(None=FunctionInput() {inputType=FunctionInput_MaterialAttributes}) {attributeGetTypes [#1, #2]}
  n2 = NamedRerouteUsage()
  n3 = Multiply(A=NamedRerouteUsage(), B=VectorParameter() {group=MistRaven, parameterName=MagicSmokeColor, defaultValue=<rgba 5 0 0.0395 1>}.RGB)
  n4 = TextureCoordinate()
  n5 = TextureSample(UVs=Panner(Coordinate=Multiply(A=@n4), Time=NamedRerouteUsage()) {speedY=-0.1}) {
   samplerSource=SSM_Wrap_WorldGroupSettings
   samplerType=SAMPLERTYPE_Grayscale
   texture=/Game/Global/Textures/Default/Noise/LowResBlurredNoiseNormalized
  }
  n6 = Multiply(A=LinearInterpolate(Alpha=@n5.R) {constA=0.75}, B=Multiply(A=NamedRerouteUsage()) {constB=0.85}) {constA=0.7}
  n7 = Constant() {r=0.8}
  n8 = TextureCoordinate()
  n9 = NamedRerouteUsage()
  n10 = TextureSample(UVs=Panner(Coordinate=Multiply(A=@n8) {constB=2}, Time=@n9) {speedY=-0.4}) {
   samplerSource=SSM_Wrap_WorldGroupSettings
   samplerType=SAMPLERTYPE_Grayscale
   texture=&LowResBlurredNoiseNormalized
  }
  n11 = TextureCoordinate()
  n12 = TextureSample(UVs=Panner(Coordinate=Multiply(A=@n11), Time=@n9) {speedY=-0.05}) {
   samplerSource=SSM_Wrap_WorldGroupSettings
   samplerType=SAMPLERTYPE_Grayscale
   texture=&LowResBlurredNoiseNormalized
  }
  n13 = Power(Base=Max(A=@n10.R, B=@n12.R)) {constExponent=1.5}
  n14 = Multiply(
   A=NamedRerouteUsage()
   B=Multiply(
    A=LinearInterpolate(A=@n7, Alpha=@n13) {constA=0.8}
    B=VectorParameter() {group=MistRaven, parameterName=Base Smoke Color, defaultValue=<rgba 4 0 0.2377 1>}.RGB
   )
  ) {constB=5}
  n15 = TextureCoordinate()
  n16 = TextureSample(UVs=Panner(Coordinate=Multiply(A=@n15) {constB=2}, Time=NamedRerouteUsage()) {speedY=-0.1}) {
   samplerSource=SSM_Wrap_WorldGroupSettings
   samplerType=SAMPLERTYPE_Grayscale
   texture=&LowResBlurredNoiseNormalized
  }
  n17 = Add(A=NamedRerouteUsage(), B=Multiply(A=Max(A=@n6, B=@n14), B=LinearInterpolate(Alpha=@n16.R) {constA=0.8}))
  n18 = TextureCoordinate()
  n19 = Constant2Vector() {g=2, r=4}
  n20 = TextureSample(UVs=Panner(Coordinate=Multiply(A=@n18, B=@n19) {constB=2}, Time=NamedRerouteUsage()) {speedX=0.2, speedY=-0.2}) {
   samplerSource=SSM_Wrap_WorldGroupSettings
   samplerType=SAMPLERTYPE_Grayscale
   texture=&LowResBlurredNoiseNormalized
  }
  n21 = Constant() {r=1}
  n22 = Power(Base=LinearInterpolate(B=Multiply(A=@n20.R, B=@n21), Alpha=NamedRerouteUsage()) {constA=1, constB=0.7}, Exp=Constant() {r=15}) {constExponent=10}
  n23 = MaterialFunctionCall(Emissive=Add(A=Multiply(A=@n2, B=@n3), B=Multiply(A=@n17, B=@n22))) {materialFunction=&MF_FX_EmissiveAdjustments}
  n24 = Constant() {r=0.5}
  n25 = MaterialFunctionCall(Value_Morning=@n24, Value_Day=Constant() {r=0.6}, Value_Evening=@n24, Value_Night=@n24) {materialFunction=&Scalar_Time_of_Day}
  n26 = NamedRerouteUsage()
  n27 = LinearInterpolate(Alpha=NamedRerouteUsage()) {constA=1, constB=0.5}
  FunctionOutput(None=SetMaterialAttributes!(
    MaterialAttributes=@n1.MaterialAttributes
    EmissiveColor=Multiply(A=@n23.Result, B=@n25.Result) {constB=0.6}
    BaseColor=Multiply(A=@n26, B=@n27)
   ) {attributeSetTypes [#1, #2]})
  n28 = Multiply(A=NamedRerouteUsage(), B=Constant() {r=-0.1})
  n29 = Multiply(A=TextureCoordinate(), B=Constant() {r=3}) {constB=4}
  n30 = TextureSample(UVs=Multiply(A=Constant() {r=0.9}, B=Add(A=@n28, B=@n29))) {
   samplerSource=SSM_Wrap_WorldGroupSettings
   samplerType=SAMPLERTYPE_LinearColor
   texture=/Game/VisualThreatManager/StormVisuals/LargeClouds/T_Swirl_Flowmap
  }
  n31 = ComponentMask(None=@n30.RGB) {g=true, r=true}
  NamedRerouteDeclaration(None=Multiply(A=Subtract(A=@n31) {constB=0.5}, B=Constant() {r=0.2})) {name=Swirl UV, nodeColor=<rgba 0 1 0.9297 1>}
  n32 = TextureSampleParameter2D() {
   group=MistRaven
   parameterName=Smoke_detailA
   samplerType=SAMPLERTYPE_Masks
   texture=/BRCosmetics/Accessories/FORT_Backpacks/Backpack_F_MED_MistRaven/Textures/T_MistRaven_Backpack_TAMask
  }
  n33 = TextureCoordinate()
  n34 = Constant() {r=2}
  n35 = TextureSample(UVs=Panner(Coordinate=Multiply(A=@n33, B=@n34) {constB=2}) {speedX=-0.04, speedY=-0.2}) {
   samplerSource=SSM_Wrap_WorldGroupSettings
   samplerType=SAMPLERTYPE_Grayscale
   texture=&LowResBlurredNoiseNormalized
  }
  n36 = Power(Base=@n32.R, Exp=LinearInterpolate(A=Constant() {r=4}, Alpha=@n35.R))
  n37 = TextureCoordinate()
  n38 = Constant() {r=4}
  n39 = Add(A=Multiply(A=NamedRerouteUsage(), B=Constant() {r=5}), B=Panner(Coordinate=Multiply(A=@n37, B=@n38) {constB=2}) {speedX=-0.04, speedY=-0.3})
  n40 = Saturate(None=Power(
    Base=TextureSample(UVs=@n39) {samplerSource=SSM_Wrap_WorldGroupSettings, texture=/Game/Global/Textures/Default/Noise/T_FireNoise_01}.R
    Exp=Constant() {r=6}
   ) {constExponent=15})
  NamedRerouteDeclaration(None=Saturate(None=Multiply(A=@n36, B=@n40))) {name=brightSmoke, nodeColor=<rgba 0.8281 1 0 1>}
  n41 = TextureCoordinate()
  n42 = Constant() {r=2}
  n43 = TextureSample(UVs=Panner(Coordinate=Multiply(A=@n41, B=@n42), Time=NamedRerouteUsage()) {speedX=-0.2, speedY=-0.1}) {
   samplerSource=SSM_Wrap_WorldGroupSettings
   samplerType=SAMPLERTYPE_Grayscale
   texture=&LowResBlurredNoiseNormalized
  }
  n44 = TextureSampleParameter2D() {
   group=MistRaven
   parameterName=Smoke_detailA
   samplerType=SAMPLERTYPE_Masks
   texture=&T_MistRaven_Backpack_TAMask
  }
  n45 = Multiply(A=Power(Base=@n43.R), B=Power(Base=@n44.R, Exp=Constant() {r=2}))
  n46 = VectorParameter() {group=MistRaven, parameterName=BurnEdgeColor2, defaultValue=<rgba 1.5 0 0.1842 1>}
  n47 = TextureCoordinate()
  n48 = Constant2Vector() {g=4, r=8}
  n49 = TextureSample(UVs=Panner(Coordinate=Multiply(A=@n47, B=@n48) {constB=2}) {speedX=0.02, speedY=-0.02}) {
   samplerSource=SSM_Wrap_WorldGroupSettings
   samplerType=SAMPLERTYPE_Grayscale
   texture=&LowResBlurredNoiseNormalized
  }
  n50 = Power(Base=@n49.R) {constExponent=1.7}
  n51 = TextureSampleParameter2D() {
   group=MistRaven
   parameterName=T_Mask
   samplerType=SAMPLERTYPE_Masks
   texture=/BRCosmetics/Accessories/FORT_Backpacks/Backpack_F_MED_MistRaven/Textures/T_MistRaven_Backpack_M
  }
  n52 = OneMinus(None=Multiply(A=@n51.G))
  n53 = Power(Base=@n52) {constExponent=6}
  n54 = Add(A=Multiply(A=Multiply(A=@n50, B=@n53), B=Constant() {r=20}) {constB=5}, B=Power(Base=@n52) {constExponent=5})
  n55 = Time()
  n56 = Constant() {r=0.1}
  n57 = Add(A=Sine(None=Multiply(A=@n55, B=@n56) {constB=0.1}))
  n58 = TextureCoordinate()
  n59 = TextureSample(UVs=Panner(Coordinate=Multiply(A=@n58), Time=Time()) {speedY=-0.1}) {
   samplerSource=SSM_Wrap_WorldGroupSettings
   samplerType=SAMPLERTYPE_Grayscale
   texture=&LowResBlurredNoiseNormalized
  }
  n60 = Multiply(
   A=Multiply(A=Divide(A=@n57), B=VectorParameter() {group=MistRaven, parameterName=BurnEdgeColor, defaultValue=<rgba 1 0 0.0201 1>}.RGB)
   B=@n59.R
  )
  NamedRerouteDeclaration(None=Add(A=Multiply(A=@n45, B=@n46.RGB), B=Multiply(A=@n54, B=@n60))) {name=BurnEdge, nodeColor=<rgba 0.7109 1 0 1>}
  n61 = TextureSampleParameter2D() {
   group=MistRaven
   parameterName=Smoke_detailA
   samplerType=SAMPLERTYPE_Masks
   texture=&T_MistRaven_Backpack_TAMask
  }
  n62 = NamedRerouteUsage()
  n63 = Multiply(A=TextureCoordinate(), B=ScalarParameter() {defaultValue=4, group=MistRaven, parameterName=BlackSmokeTiling}) {constB=2}
  n64 = NamedRerouteUsage()
  n65 = TextureSampleParameter2D(UVs=Add(A=Multiply(A=@n62) {constB=3}, B=Panner(Coordinate=@n63, Time=@n64) {speedX=-0.04, speedY=-0.4})) {
   group=MistRaven
   parameterName=Smoke_detailA
   samplerType=SAMPLERTYPE_Masks
   texture=&T_MistRaven_Backpack_TAMask
  }
  n66 = Power(Base=@n65.B) {constExponent=4}
  n67 = Multiply(A=Power(Base=@n61.R, Exp=Constant() {r=1.8}), B=Saturate(None=Multiply(A=@n66) {constB=3}))
  n68 = SphereMask(A=VertexInterpolator(VS=LocalPosition() {localOrigin=InstancePreSkinning}.XYZ).PS, B=Constant()) {attenuationRadius=80, hardnessPercent=0}
  NamedRerouteDeclaration(None=Saturate(None=Multiply(A=@n67, B=@n68))) {name=BlackSmoke, nodeColor=<rgba 0.6406 1 0 1>}
  NamedRerouteDeclaration(None=@n1.BaseColor) {name=InPutBaseColor, nodeColor=<rgba 0 1 0.7188 1>}
  NamedRerouteDeclaration(None=@n1.EmissiveColor) {name=InputEmissive, nodeColor=<rgba 0.7109 0 1 1>}
  MaterialExpressionVertexInterpolator_0 = VertexInterpolator(VS=LocalPosition() {localOrigin=InstancePreSkinning}.XYZ)
  n69 = Convert(None=@MaterialExpressionVertexInterpolator_0.PS) {
   nodeName=Break Float3
   convertInputs [{
     type=Vector3
     expressionInput {
      expression=&self:MaterialExpressionVertexInterpolator_0
      inputName=None
      mask=0
      maskA=0
      maskB=0
      maskG=0
      maskR=0
      outputIndex=0
     }
    }]
   convertOutputs [{type=Scalar}, {type=Scalar}, {type=Scalar}]
  }
  n70 = Constant()
  NamedRerouteDeclaration(None=If(A=Subtract(A=@n69) {constB=0}, "A > B"=Constant() {r=1}, "A == B"=@n70, "A < B"=@n70)) {name=WingMask, nodeColor=<rgba 0.5 0 1 1>}
  n71 = Time()
  n72 = Constant() {r=2}
  NamedRerouteDeclaration(None=LinearInterpolate(A=Time(), B=Add(A=@n71, B=@n72) {constB=2}, Alpha=NamedRerouteUsage())) {name=TimeOffset, nodeColor=<rgba 1 0.2578 0 1>}
  n73 = Multiply(A=TextureCoordinate(), B=ScalarParameter() {defaultValue=3, group=MistRaven, parameterName=BaseSmokeTiling}) {constB=5}
  n74 = NamedRerouteUsage()
  n75 = AppendVector(
   A=ScalarParameter() {defaultValue=0.04, group=MistRaven, parameterName=BaseSmokeSpeedX}
   B=ScalarParameter() {defaultValue=-0.2, group=MistRaven, parameterName=BaseSmokeSpeedY}
  )
  n76 = TextureSampleParameter2D(UVs=Add(A=NamedRerouteUsage(), B=Panner(Coordinate=@n73, Time=@n74, Speed=@n75) {speedX=0.02, speedY=-0.1})) {
   group=MistRaven
   parameterName=Smoke_detailA
   samplerType=SAMPLERTYPE_Masks
   texture=&T_MistRaven_Backpack_TAMask
  }
  NamedRerouteDeclaration(None=@n76.B) {name=Base Smoke, nodeColor=<rgba 1 0.6797 0 1>}
  n77 = TextureCoordinate()
  n78 = AppendVector(
   A=ScalarParameter() {defaultValue=15, group=MistRaven, parameterName=NoiseUV_X}
   B=ScalarParameter() {defaultValue=15, group=MistRaven, parameterName=NoiseUV_Y}
  )
  n79 = TextureSample(UVs=Panner(Coordinate=Multiply(A=@n77, B=@n78) {constB=15}) {speedY=-0.4}) {
   samplerSource=SSM_Wrap_WorldGroupSettings
   samplerType=SAMPLERTYPE_Grayscale
   texture=&LowResBlurredNoiseNormalized
  }
  NamedRerouteDeclaration(None=@n79.G) {name=NoiseUv, nodeColor=<rgba 0.3594 0 1 1>}
  n80 = Sine(None=Multiply(A=NamedRerouteUsage(), B=Constant() {r=0.1}) {constB=0.1})
  NamedRerouteDeclaration(None=Divide(A=Add(A=@n80))) {name=FlickerAnimation1, nodeColor=<rgba 1 0.1172 0 1>}
  n81 = NamedRerouteUsage()
  n82 = Constant() {r=0.005}
  n83 = TextureSampleParameter2D(UVs=Add(A=TextureCoordinate(), B=Multiply(A=@n81, B=@n82) {constB=0.01})) {
   group=MistRaven
   parameterName=Smoke_detailA
   samplerType=SAMPLERTYPE_Masks
   texture=&T_MistRaven_Backpack_TAMask
  }
  n84 = TextureCoordinate()
  n85 = TextureSample(UVs=Panner(Coordinate=Multiply(A=@n84) {constB=2}, Time=NamedRerouteUsage()) {speedX=0.05, speedY=-0.1}) {
   samplerSource=SSM_Wrap_WorldGroupSettings
   samplerType=SAMPLERTYPE_Grayscale
   texture=&LowResBlurredNoiseNormalized
  }
  n86 = LinearInterpolate(Alpha=@n85.R)
  n87 = NamedRerouteUsage()
  n88 = Constant() {r=0.004}
  n89 = TextureSampleParameter2D(UVs=Add(A=TextureCoordinate(), B=Multiply(A=@n87, B=@n88) {constB=0.01})) {
   group=MistRaven
   parameterName=Smoke_detailA
   samplerType=SAMPLERTYPE_Masks
   texture=&T_MistRaven_Backpack_TAMask
  }
  NamedRerouteDeclaration(None=Max(A=Multiply(A=@n83.G, B=@n86), B=@n89.R)) {name=SmokeMask, nodeColor=<rgba 0 1 0.6484 1>}
  n90 = NamedRerouteUsage()
  n91 = Sine(None=Multiply(A=Time(), B=Constant() {r=0.1}) {constB=0.1})
  n92 = LinearInterpolate(Alpha=Divide(A=Add(A=@n91))) {constA=1.4, constB=1.1}
  n93 = NamedRerouteUsage()
  n94 = Multiply(A=TextureCoordinate(), B=ScalarParameter() {defaultValue=5, group=MistRaven, parameterName=DetailSmoketiling2}) {constB=2}
  n95 = NamedRerouteUsage()
  n96 = AppendVector(
   A=ScalarParameter() {defaultValue=-0.04, group=MistRaven, parameterName=DetailSmoke_X}
   B=ScalarParameter() {defaultValue=-0.4, group=MistRaven, parameterName=DetailSmoke_Y}
  )
  n97 = Add(
   A=Multiply(A=NamedRerouteUsage()) {constB=0.05}
   B=Add(A=Multiply(A=@n93) {constB=2}, B=Panner(Coordinate=@n94, Time=@n95, Speed=@n96) {speedX=-0.02, speedY=-0.25})
  )
  n98 = Power(Base=Max(A=TextureSample(UVs=@n97) {texture=/Game/Global/Textures/Default/Noise/T_FireNoise_02}.R, B=NamedRerouteUsage())) {constExponent=3}
  NamedRerouteDeclaration(None=Multiply(A=Power(Base=@n90, Exp=@n92) {constExponent=1.2}, B=Saturate(None=@n98))) {name=DetailSmoke, nodeColor=<rgba 1 0.3281 0 1>}
 ]}
