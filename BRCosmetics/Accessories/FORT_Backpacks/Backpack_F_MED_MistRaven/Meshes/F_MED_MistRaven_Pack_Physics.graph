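# Physics asset for the MistRaven backpack cosmetic, giving its left and right wings collision bodies and joints for physical motion.
# Each side chains a WingBase sphere through Upper, lower and Hand capsules, then fans into four feather chains (BigFeather, A, B, C).
# Constraints set narrow swing and twist limits along each chain, while the BigFeatherB tips have swing and twist locked to their parents.
# The two WingBase bodies have no constraint joining them.
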PhysicsAsset /BRCosmetics/Accessories/FORT_Backpacks/Backpack_F_MED_MistRaven/Meshes/F_MED_MistRaven_Pack_Physics
bodies {
 BigFeatherA_L [{shapeType=Capsule, length=9.2587, radius=2.5726, shapeName=Hand_L_capsule, center=<9.3772 0 0>, rotation=<rot -90 -179.9472 179.9475>}]
 BigFeatherA_R [{shapeType=Capsule, length=9.2587, radius=2.5726, shapeName=Hand_R_capsule, center=<-9.3772 0 0>, rotation=<rot -90 -90 -89.9998>}]
 BigFeatherB_L [{shapeType=Capsule, length=5.7643, radius=2.6021, shapeName=Hand_L_capsule, center=<5.5436 0 0>, rotation=<rot -90 -179.9472 179.9475>}]
 BigFeatherB_R [{shapeType=Capsule, length=5.7643, radius=2.6021, shapeName=Hand_R_capsule, center=<-5.5436 0 0>, rotation=<rot -90 -90 -89.9998>}]
 BigFeather_L [{shapeType=Capsule, length=8.4532, radius=3.3782, shapeName=Hand_L_capsule, center=<9.3772 0 0>, rotation=<rot -90 -179.9472 179.9475>}]
 BigFeather_R [{shapeType=Capsule, length=8.4532, radius=3.3782, shapeName=Hand_R_capsule, center=<-9.3772 0 0>, rotation=<rot -90 -90 -89.9998>}]
 FeatherA1_L [{shapeType=Capsule, length=9.5235, radius=2.3079, shapeName=Hand_L_capsule, center=<7.6633 0 0>, rotation=<rot -90 -144.7357 144.736>}]
 FeatherA1_R [{shapeType=Capsule, length=9.5235, radius=2.3079, shapeName=Hand_R_capsule, center=<-7.6633 0 0>, rotation=<rot -90 -174.9488 -5.0509>}]
 FeatherA2_L [{shapeType=Capsule, length=3.1766, radius=2.2632, shapeName=Hand_L_capsule, center=<4.3622 0 0>, rotation=<rot -90 -144.7357 144.736>}]
 FeatherA2_R [{shapeType=Capsule, length=3.1766, radius=2.2632, shapeName=Hand_R_capsule, center=<-4.3622 0 0>, rotation=<rot -90 -174.9488 -5.0509>}]
 FeatherA_L [{shapeType=Capsule, length=9.2115, radius=2.6199, shapeName=Hand_L_capsule, center=<7.426 0 0.5973>, rotation=<rot -90 -178.7341 178.7344>}]
 FeatherA_R [{shapeType=Capsule, length=9.2115, radius=2.6199, shapeName=Hand_R_capsule, center=<-7.426 0 -0.5973>, rotation=<rot -90 -179.6835 -0.3163>}]
 FeatherB1_L [{shapeType=Capsule, length=5.8206, radius=2.2632, shapeName=Hand_L_capsule, center=<7.3863 0 0>, rotation=<rot -90 -174.9488 174.9491>}]
 FeatherB1_R [{shapeType=Capsule, length=5.8206, radius=2.2632, shapeName=Hand_R_capsule, center=<-7.3863 0 0>, rotation=<rot -90 -178.7341 -1.2656>}]
 FeatherB_L [{shapeType=Capsule, length=9.2115, radius=2.6199, shapeName=Hand_L_capsule, center=<7.426 0 1.1192>, rotation=<rot -90 -144.7357 144.736>}]
 FeatherB_R [{shapeType=Capsule, length=9.2115, radius=2.6199, shapeName=Hand_R_capsule, center=<-7.426 0 -1.1192>, rotation=<rot -90 -174.9488 -5.0509>}]
 FeatherC1_L [{shapeType=Capsule, length=3.1766, radius=2.2632, shapeName=Hand_L_capsule, center=<4.3622 0 0>, rotation=<rot -90 -144.7357 144.736>}]
 FeatherC1_R [{shapeType=Capsule, length=3.1766, radius=2.2632, shapeName=Hand_R_capsule, center=<-4.3622 0 0>, rotation=<rot -90 -174.9488 -5.0509>}]
 FeatherC_L [{shapeType=Capsule, length=6.4871, radius=2.2632, shapeName=Hand_L_capsule, center=<5.4002 0 0>, rotation=<rot -90 -174.9488 174.9491>}]
 FeatherC_R [{shapeType=Capsule, length=6.4871, radius=2.2632, shapeName=Hand_R_capsule, center=<-5.4002 0 0>, rotation=<rot -90 -178.7341 -1.2656>}]
 Hand_L [{
   shapeType=Capsule
   length=7.8538
   radius=3.9775
   shapeName=Hand_L_capsule
   center=<-13.269 1.7377 -13.7306>
   rotation=<rot 1.8941 -96.6871 -119.9993>
  }]
 Hand_R [{
   shapeType=Capsule
   length=7.8538
   radius=3.9775
   shapeName=Hand_R_capsule
   center=<13.269 -1.7377 13.7306>
   rotation=<rot 1.8941 -96.6871 60.0007>
  }]
 Upper_L [{
   shapeType=Capsule
   length=5.2113
   radius=3.8245
   shapeName=Upper_L_capsule
   center=<3.4535 -0.1361 -0.2145>
   rotation=<rot 0.0001 93.0522 -86.3702>
  }]
 Upper_R [{
   shapeType=Capsule
   length=5.2113
   radius=3.8245
   shapeName=Upper_R_capsule
   center=<-3.4535 0.1361 0.2145>
   rotation=<rot 0.0001 93.0522 93.6298>
  }]
 WingBase_L [{shapeType=Sphere, radius=2, shapeName=WingBase_L_sphere}]
 WingBase_R [{shapeType=Sphere, radius=2, shapeName=WingBase_R_sphere}]
 lower_L [{
   shapeType=Capsule
   length=5.4443
   radius=3.2617
   shapeName=lower_L_capsule
   center=<13.9211 0.2509 -0.5622>
   rotation=<rot 8.0043 -89.9999 -87.591>
  }]
 lower_R [{
   shapeType=Capsule
   length=5.4443
   radius=3.2617
   shapeName=lower_R_capsule
   center=<-13.9211 -0.2509 0.5622>
   rotation=<rot 8.0043 -89.9999 92.409>
  }]
}
constraints [
 {bone1Name=lower_L, bone2Name=Upper_L, swing1LimitDegrees=5, swing2LimitDegrees=5, twistLimitDegrees=10}
 {bone1Name=lower_R, bone2Name=Upper_R, swing1LimitDegrees=5, swing2LimitDegrees=5, twistLimitDegrees=10}
 {bone1Name=Hand_L, bone2Name=lower_L, swing1LimitDegrees=15, swing2LimitDegrees=0, twistLimitDegrees=25}
 {bone1Name=Hand_R, bone2Name=lower_R, swing1LimitDegrees=15, swing2LimitDegrees=0, twistLimitDegrees=25}
 {bone1Name=BigFeather_L, bone2Name=Hand_L, swing1LimitDegrees=15, swing2LimitDegrees=7, twistLimitDegrees=2}
 {bone1Name=BigFeatherA_L, bone2Name=BigFeather_L, swing1LimitDegrees=15, swing2LimitDegrees=7, twistLimitDegrees=2}
 {
  bone1Name=BigFeatherB_L
  bone2Name=BigFeatherA_L
  swing1LimitDegrees=5
  swing1Motion=Locked
  swing2LimitDegrees=5
  swing2Motion=Locked
  twistLimitDegrees=2
  twistMotion=Locked
 }
 {bone1Name=FeatherC_L, bone2Name=Hand_L, swing1LimitDegrees=15, swing2LimitDegrees=7, twistLimitDegrees=2}
 {bone1Name=FeatherC1_L, bone2Name=FeatherC_L, swing1LimitDegrees=15, swing2LimitDegrees=7, twistLimitDegrees=2}
 {bone1Name=FeatherA_L, bone2Name=Hand_L, swing1LimitDegrees=15, swing2LimitDegrees=7, twistLimitDegrees=2}
 {bone1Name=FeatherA1_L, bone2Name=FeatherA_L, swing1LimitDegrees=15, swing2LimitDegrees=7, twistLimitDegrees=2}
 {bone1Name=FeatherA2_L, bone2Name=FeatherA1_L, swing1LimitDegrees=15, swing2LimitDegrees=7, twistLimitDegrees=2}
 {bone1Name=FeatherB_L, bone2Name=Hand_L, swing1LimitDegrees=15, swing2LimitDegrees=7, twistLimitDegrees=2}
 {bone1Name=FeatherB1_L, bone2Name=FeatherB_L, swing1LimitDegrees=15, swing2LimitDegrees=7, twistLimitDegrees=2}
 {bone1Name=FeatherA_R, bone2Name=Hand_R, swing1LimitDegrees=15, swing2LimitDegrees=7, twistLimitDegrees=2}
 {bone1Name=FeatherA1_R, bone2Name=FeatherA_R, swing1LimitDegrees=15, swing2LimitDegrees=7, twistLimitDegrees=2}
 {bone1Name=FeatherA2_R, bone2Name=FeatherA1_R, swing1LimitDegrees=15, swing2LimitDegrees=7, twistLimitDegrees=2}
 {bone1Name=FeatherB_R, bone2Name=Hand_R, swing1LimitDegrees=15, swing2LimitDegrees=7, twistLimitDegrees=2}
 {bone1Name=FeatherB1_R, bone2Name=FeatherB_R, swing1LimitDegrees=15, swing2LimitDegrees=7, twistLimitDegrees=2}
 {bone1Name=FeatherC_R, bone2Name=Hand_R, swing1LimitDegrees=15, swing2LimitDegrees=7, twistLimitDegrees=2}
 {bone1Name=FeatherC1_R, bone2Name=FeatherC_R, swing1LimitDegrees=15, swing2LimitDegrees=7, twistLimitDegrees=2}
 {bone1Name=BigFeather_R, bone2Name=Hand_R, swing1LimitDegrees=15, swing2LimitDegrees=7, twistLimitDegrees=2}
 {bone1Name=BigFeatherA_R, bone2Name=BigFeather_R, swing1LimitDegrees=15, swing2LimitDegrees=7, twistLimitDegrees=2}
 {
  bone1Name=BigFeatherB_R
  bone2Name=BigFeatherA_R
  swing1LimitDegrees=5
  swing1Motion=Locked
  swing2LimitDegrees=5
  swing2Motion=Locked
  twistLimitDegrees=2
  twistMotion=Locked
 }
 {bone1Name=Upper_L, bone2Name=WingBase_L, swing1LimitDegrees=19.08, swing2LimitDegrees=0, twistLimitDegrees=5}
 {bone1Name=Upper_R, bone2Name=WingBase_R, swing1LimitDegrees=19.08, swing2LimitDegrees=0, twistLimitDegrees=5}
]
